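# Modular Control Rig face module for Sidekick that provides a slider board for animating facial pose curves.
# Construction spawns a scaled Face Rig control with limited 2D child controls for jaw, tongue, lips, eyes, brows, cheeks and nose, plus optional pupil controls.
# The rig graph converts control axes to pose curves, often splitting signed axes into positive/negative pairs.
# Smile-open correctives take the minimum of smile and jaw-open values.
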ControlRigBlueprint /EpicControlRigPub/v5_8/Modules/Sidekick/Face/CRM_Epic_Sidekick_Face_v01
graphs [
 ConstructionGraph {nodes [
   RigUnit_PrepareForExecution = &RigUnit_PrepareForExecution()
   RigVMFunction_Sequence = RigVMAggregateNode(ExecuteContext=@RigUnit_PrepareForExecution.ExecutePin)
   Get_Module_Metadata_3 = RigVMDispatchNode(Default=1, Name=Global Control Scale, NameSpace=Root) {template="DISPATCH_RigDispatch_GetModuleMetadata(in Name,in NameSpace,in Default,out Value,out Found)"}
   Multiply = &RigVMFunction_MathDoubleMul(A=@Get_Module_Metadata_3.Value, B=RigVMVariableNode() {variable=Control Scale}.Value)
   n1 = RigVMVariableNode() {variable=Face Rig Offset Transform}
   Scale = /Script/RigVM.RigVMFunction_MathVectorScale(Factor=@Multiply.Result, Value=@n1.Value.Scale3D) {is=struct RigVMUnitNode, template="Scale::Execute(in Value,in Factor,out Result)"}
   n2 = /Script/ControlRig.RigUnit_HierarchyAddControlTransform(
    InitialValue=(Rotation=(X=0,Y=0,Z=0,W=1),Translation=(X=0,Y=0,Z=0),Scale3D=(X=1,Y=1,Z=1))
    Name=Face Rig
    OffsetSpace=GlobalSpace
    OffsetTransform=(Rotation=(X=0,Y=0,Z=0,W=1),Translation=(X=0,Y=0,Z=-30),Scale3D=(X=1,Y=1,Z=1))
    Parent=(Type=Bone,Name="None")
    Settings="(InitialSpace=LocalSpace,bUsePreferredRotationOrder=False,PreferredRotationOrder=YZX,Limits=(LimitTranslationX=(bMinimum=False,bMaximum=False),LimitTranslationY=(bMinimum=False,bMaximum=False),LimitTranslationZ=(bMinimum=False,bMaximum=False),LimitPitch=(b ...(truncated)"
    OffsetTransform.Scale3D=@Scale.Result
    OffsetTransform.Rotation=@n1.Value.Rotation
    OffsetTransform.Translation=@n1.Value.Translation
    ExecutePin=@RigVMFunction_Sequence.B
   ) {
    is=struct RigVMUnitNode
    template="SpawnControl::Execute(in InitialValue,in Settings,in OffsetTransform,in OffsetSpace,in Parent,in Name,out Item)"
   }
   > RigVMVariableNode(Value=@n2.Item) {variable=Face Rig Control}
   SpawnControl = /Script/ControlRig.RigUnit_HierarchyAddControlVector2D(
    InitialValue=(X=0,Y=0)
    Name=Jaw
    OffsetSpace=LocalSpace
    OffsetTransform=(Rotation=(X=-0.7071,Y=-0.7071,Z=0,W=0),Translation=(X=0,Y=0,Z=-14),Scale3D=(X=4,Y=4,Z=4))
    Settings="(PrimaryAxis=X,Limits=(LimitX=(bMinimum=True,bMaximum=True),LimitY=(bMinimum=True,bMaximum=True),MinValue=(X=-1,Y=-1),MaxValue=(X=1,Y=1),bDrawLimits=True),Shape=(bVisible=True,Name=\"Circle_Solid\",Color=(R=1,G=1,B=0,A=1),Transform=(Rotation=(X=-0.5,Y=-0.5,Z ...(truncated)"
    Parent=RigVMVariableNode() {variable=Face Rig Control}.Value
    ExecutePin=@RigVMFunction_Sequence.C
   ) {
    is=struct RigVMUnitNode
    template="SpawnControl::Execute(in InitialValue,in Settings,in OffsetTransform,in OffsetSpace,in Parent,in Name,out Item)"
   }
   > RigVMVariableNode(Value=@SpawnControl.Item) {variable=Jaw Main Control}
   > n3 = &RigUnit_HierarchyAddControlVector2D(
    InitialValue=(X=0,Y=0)
    Name=Tongue
    OffsetSpace=LocalSpace
    OffsetTransform=(Rotation=(X=-0.7071,Y=-0.7071,Z=0,W=0),Translation=(X=8,Y=0,Z=-13),Scale3D=(X=2,Y=2,Z=2))
    Settings="(PrimaryAxis=X,Limits=(LimitX=(bMinimum=True,bMaximum=True),LimitY=(bMinimum=True,bMaximum=True),MinValue=(X=-1,Y=-1),MaxValue=(X=1,Y=1),bDrawLimits=True),Shape=(bVisible=True,Name=\"Circle_Solid\",Color=(R=1,G=1,B=0,A=1),Transform=(Rotation=(X=-0.5,Y=-0.5,Z ...(truncated)"
    Parent=RigVMVariableNode() {variable=Face Rig Control}.Value
   ) {template="SpawnControl::Execute(in InitialValue,in Settings,in OffsetTransform,in OffsetSpace,in Parent,in Name,out Item)"}
   > RigVMVariableNode(Value=@n3.Item) {variable=Tongue Control}
   > n4 = &RigUnit_HierarchyAddControlVector2D(
    InitialValue=(X=0,Y=0)
    Name=Tongue Fwd
    OffsetSpace=LocalSpace
    OffsetTransform=(Rotation=(X=0.7071,Y=0,Z=0,W=0.7071),Translation=(X=11,Y=0,Z=-13),Scale3D=(X=2,Y=2,Z=2))
    Settings="(PrimaryAxis=X,Limits=(LimitX=(bMinimum=True,bMaximum=True),LimitY=(bMinimum=True,bMaximum=True),MinValue=(X=-1,Y=0),MaxValue=(X=1,Y=0),bDrawLimits=True),Shape=(bVisible=True,Name=\"Circle_Solid\",Color=(R=1,G=1,B=0,A=1),Transform=(Rotation=(X=0,Y=0,Z=0,W=1) ...(truncated)"
    Parent=RigVMVariableNode() {variable=Face Rig Control}.Value
   ) {template="SpawnControl::Execute(in InitialValue,in Settings,in OffsetTransform,in OffsetSpace,in Parent,in Name,out Item)"}
   > RigVMVariableNode(Value=@n4.Item) {variable=Tongue Fwd Control}
   > n5 = &RigUnit_HierarchyAddControlVector2D(
    InitialValue=(X=0,Y=0)
    Name=Tongue Twist
    OffsetSpace=LocalSpace
    OffsetTransform=(Rotation=(X=0.5,Y=-0.5,Z=0.5,W=0.5),Translation=(X=8,Y=0,Z=-10),Scale3D=(X=2,Y=2,Z=2))
    Settings="(PrimaryAxis=X,Limits=(LimitX=(bMinimum=True,bMaximum=True),LimitY=(bMinimum=True,bMaximum=True),MinValue=(X=-1,Y=0),MaxValue=(X=1,Y=0),bDrawLimits=True),Shape=(bVisible=True,Name=\"Circle_Solid\",Color=(R=1,G=1,B=0,A=1),Transform=(Rotation=(X=0,Y=-1,Z=0,W=0 ...(truncated)"
    Parent=RigVMVariableNode() {variable=Face Rig Control}.Value
   ) {template="SpawnControl::Execute(in InitialValue,in Settings,in OffsetTransform,in OffsetSpace,in Parent,in Name,out Item)"}
   > RigVMVariableNode(Value=@n5.Item) {variable=Tongue Twist Control}
   > n6 = &RigUnit_HierarchyAddControlVector2D(
    InitialValue=(X=0,Y=0)
    Name=Tongue Curl Tip
    OffsetSpace=LocalSpace
    OffsetTransform=(Rotation=(X=0.7071,Y=0,Z=0,W=0.7071),Translation=(X=9,Y=0,Z=-17.5),Scale3D=(X=2,Y=2,Z=2))
    Settings="(PrimaryAxis=X,Limits=(LimitX=(bMinimum=True,bMaximum=True),LimitY=(bMinimum=True,bMaximum=True),MinValue=(X=-1,Y=0),MaxValue=(X=1,Y=0),bDrawLimits=True),Shape=(bVisible=True,Name=\"Circle_Solid\",Color=(R=1,G=1,B=0,A=1),Transform=(Rotation=(X=0,Y=0,Z=0,W=1) ...(truncated)"
    Parent=RigVMVariableNode() {variable=Face Rig Control}.Value
   ) {template="SpawnControl::Execute(in InitialValue,in Settings,in OffsetTransform,in OffsetSpace,in Parent,in Name,out Item)"}
   > RigVMVariableNode(Value=@n6.Item) {variable=Tongue Curl Tip Control}
   > n7 = &RigUnit_HierarchyAddControlVector2D(
    InitialValue=(X=0,Y=0)
    Name=Tongue Curl Base
    OffsetSpace=LocalSpace
    OffsetTransform=(Rotation=(X=0.7071,Y=0,Z=0,W=0.7071),Translation=(X=7,Y=0,Z=-17.5),Scale3D=(X=2,Y=2,Z=2))
    Settings="(PrimaryAxis=X,Limits=(LimitX=(bMinimum=True,bMaximum=True),LimitY=(bMinimum=True,bMaximum=True),MinValue=(X=-1,Y=0),MaxValue=(X=1,Y=0),bDrawLimits=True),Shape=(bVisible=True,Name=\"Circle_Solid\",Color=(R=1,G=1,B=0,A=1),Transform=(Rotation=(X=0,Y=0,Z=0,W=1) ...(truncated)"
    Parent=RigVMVariableNode() {variable=Face Rig Control}.Value
   ) {template="SpawnControl::Execute(in InitialValue,in Settings,in OffsetTransform,in OffsetSpace,in Parent,in Name,out Item)"}
   > RigVMVariableNode(Value=@n7.Item) {variable=Tongue Curl Control}
   > n8 = &RigUnit_HierarchyAddControlVector2D(
    InitialValue=(X=0,Y=0)
    Name=Reverse Jaw
    OffsetSpace=LocalSpace
    OffsetTransform=(Rotation=(X=0,Y=0,Z=0.7071,W=0.7071),Translation=(X=-8,Y=0,Z=-17),Scale3D=(X=6,Y=6,Z=6))
    Settings="(PrimaryAxis=X,Limits=(LimitX=(bMinimum=True,bMaximum=True),LimitY=(bMinimum=True,bMaximum=True),MinValue=(X=0,Y=0),MaxValue=(X=0,Y=1),bDrawLimits=True),Shape=(bVisible=True,Name=\"Circle_Solid\",Color=(R=1,G=1,B=0,A=1),Transform=(Rotation=(X=-0.5,Y=-0.5,Z=0 ...(truncated)"
    Parent=RigVMVariableNode() {variable=Face Rig Control}.Value
   ) {template="SpawnControl::Execute(in InitialValue,in Settings,in OffsetTransform,in OffsetSpace,in Parent,in Name,out Item)"}
   > RigVMVariableNode(Value=@n8.Item) {variable=Reverse Jaw Control}
   n9 = &RigUnit_HierarchyAddControlVector2D(
    InitialValue=(X=0,Y=0)
    Name=L Corner
    OffsetSpace=LocalSpace
    OffsetTransform=(Rotation=(X=0,Y=0,Z=0.7071,W=0.7071),Translation=(X=7,Y=0,Z=-4),Scale3D=(X=3,Y=3,Z=3))
    Settings="(PrimaryAxis=X,Limits=(LimitX=(bMinimum=True,bMaximum=True),LimitY=(bMinimum=True,bMaximum=True),MinValue=(X=-1,Y=-1),MaxValue=(X=1,Y=1),bDrawLimits=True),Shape=(bVisible=True,Name=\"Circle_Solid\",Color=(R=0,G=1,B=0,A=1),Transform=(Rotation=(X=-0.5,Y=-0.5,Z ...(truncated)"
    Parent=RigVMVariableNode() {variable=Face Rig Control}.Value
    ExecutePin=@RigVMFunction_Sequence.D
   ) {template="SpawnControl::Execute(in InitialValue,in Settings,in OffsetTransform,in OffsetSpace,in Parent,in Name,out Item)"}
   > RigVMVariableNode(Value=@n9.Item) {variable=L Corner Control}
   > n10 = &RigUnit_HierarchyAddControlVector2D(
    InitialValue=(X=0,Y=0)
    Name=R Corner
    OffsetSpace=LocalSpace
    OffsetTransform=(Rotation=(X=0,Y=0,Z=0.7071,W=0.7071),Translation=(X=-7,Y=0,Z=-4),Scale3D=(X=3,Y=-3,Z=3))
    Settings="(PrimaryAxis=X,Limits=(LimitX=(bMinimum=True,bMaximum=True),LimitY=(bMinimum=True,bMaximum=True),MinValue=(X=-1,Y=-1),MaxValue=(X=1,Y=1),bDrawLimits=True),Shape=(bVisible=True,Name=\"Circle_Solid\",Color=(R=1,G=0,B=0,A=1),Transform=(Rotation=(X=-0.5,Y=-0.5,Z ...(truncated)"
    Parent=RigVMVariableNode() {variable=Face Rig Control}.Value
   ) {template="SpawnControl::Execute(in InitialValue,in Settings,in OffsetTransform,in OffsetSpace,in Parent,in Name,out Item)"}
   > RigVMVariableNode(Value=@n10.Item) {variable=R Corner Control}
   > n11 = &RigUnit_HierarchyAddControlVector2D(
    InitialValue=(X=0,Y=0)
    Name=L Upper Lip
    OffsetSpace=LocalSpace
    OffsetTransform=(Rotation=(X=0,Y=0,Z=0.7071,W=0.7071),Translation=(X=2,Y=0,Z=0),Scale3D=(X=2,Y=-2,Z=2))
    Settings="(PrimaryAxis=X,Limits=(LimitX=(bMinimum=True,bMaximum=True),LimitY=(bMinimum=True,bMaximum=True),MinValue=(X=0,Y=-1),MaxValue=(X=0,Y=1),bDrawLimits=True),Shape=(bVisible=True,Name=\"Circle_Solid\",Color=(R=0,G=1,B=0,A=1),Transform=(Rotation=(X=-0.5,Y=-0.5,Z= ...(truncated)"
    Parent=RigVMVariableNode() {variable=Face Rig Control}.Value
   ) {template="SpawnControl::Execute(in InitialValue,in Settings,in OffsetTransform,in OffsetSpace,in Parent,in Name,out Item)"}
   > RigVMVariableNode(Value=@n11.Item) {variable=L Upper Lip Control}
   > n12 = &RigUnit_HierarchyAddControlVector2D(
    InitialValue=(X=0,Y=0)
    Name=R Upper Lip
    OffsetSpace=LocalSpace
    OffsetTransform=(Rotation=(X=0,Y=0,Z=0.7071,W=0.7071),Translation=(X=-2,Y=0,Z=0),Scale3D=(X=2,Y=-2,Z=2))
    Settings="(PrimaryAxis=X,Limits=(LimitX=(bMinimum=True,bMaximum=True),LimitY=(bMinimum=True,bMaximum=True),MinValue=(X=0,Y=-1),MaxValue=(X=0,Y=1),bDrawLimits=True),Shape=(bVisible=True,Name=\"Circle_Solid\",Color=(R=1,G=0,B=0,A=1),Transform=(Rotation=(X=-0.5,Y=-0.5,Z= ...(truncated)"
    Parent=RigVMVariableNode() {variable=Face Rig Control}.Value
   ) {template="SpawnControl::Execute(in InitialValue,in Settings,in OffsetTransform,in OffsetSpace,in Parent,in Name,out Item)"}
   > RigVMVariableNode(Value=@n12.Item) {variable=R Upper Lip Control}
   > n13 = &RigUnit_HierarchyAddControlVector2D(
    InitialValue=(X=0,Y=0)
    Name=L Lower LIp
    OffsetSpace=LocalSpace
    OffsetTransform=(Rotation=(X=0,Y=0,Z=0.7071,W=0.7071),Translation=(X=2,Y=0,Z=-6),Scale3D=(X=2,Y=-2,Z=2))
    Settings="(PrimaryAxis=X,Limits=(LimitX=(bMinimum=True,bMaximum=True),LimitY=(bMinimum=True,bMaximum=True),MinValue=(X=0,Y=-1),MaxValue=(X=0,Y=1),bDrawLimits=True),Shape=(bVisible=True,Name=\"Circle_Solid\",Color=(R=0,G=1,B=0,A=1),Transform=(Rotation=(X=-0.5,Y=-0.5,Z= ...(truncated)"
    Parent=RigVMVariableNode() {variable=Face Rig Control}.Value
   ) {template="SpawnControl::Execute(in InitialValue,in Settings,in OffsetTransform,in OffsetSpace,in Parent,in Name,out Item)"}
   > RigVMVariableNode(Value=@n13.Item) {variable=L Lower Lip Control}
   > n14 = &RigUnit_HierarchyAddControlVector2D(
    InitialValue=(X=0,Y=0)
    Name=R Lower LIp
    OffsetSpace=LocalSpace
    OffsetTransform=(Rotation=(X=0,Y=0,Z=0.7071,W=0.7071),Translation=(X=-2,Y=0,Z=-6),Scale3D=(X=2,Y=-2,Z=2))
    Settings="(PrimaryAxis=X,Limits=(LimitX=(bMinimum=True,bMaximum=True),LimitY=(bMinimum=True,bMaximum=True),MinValue=(X=0,Y=-1),MaxValue=(X=0,Y=1),bDrawLimits=True),Shape=(bVisible=True,Name=\"Circle_Solid\",Color=(R=1,G=0,B=0,A=1),Transform=(Rotation=(X=-0.5,Y=-0.5,Z= ...(truncated)"
    Parent=RigVMVariableNode() {variable=Face Rig Control}.Value
   ) {template="SpawnControl::Execute(in InitialValue,in Settings,in OffsetTransform,in OffsetSpace,in Parent,in Name,out Item)"}
   > RigVMVariableNode(Value=@n14.Item) {variable=R Lower Lip Control}
   > n15 = &RigUnit_HierarchyAddControlVector2D(
    InitialValue=(X=0,Y=0)
    Name=Upper Lip
    OffsetSpace=LocalSpace
    OffsetTransform=(Rotation=(X=0,Y=0,Z=0.7071,W=0.7071),Translation=(X=0,Y=0,Z=0),Scale3D=(X=2,Y=-2,Z=2))
    Settings="(PrimaryAxis=X,Limits=(LimitX=(bMinimum=True,bMaximum=True),LimitY=(bMinimum=True,bMaximum=True),MinValue=(X=0,Y=-1),MaxValue=(X=0,Y=1),bDrawLimits=True),Shape=(bVisible=True,Name=\"Circle_Solid\",Color=(R=1,G=1,B=0,A=1),Transform=(Rotation=(X=-0.5,Y=-0.5,Z= ...(truncated)"
    Parent=RigVMVariableNode() {variable=Face Rig Control}.Value
   ) {template="SpawnControl::Execute(in InitialValue,in Settings,in OffsetTransform,in OffsetSpace,in Parent,in Name,out Item)"}
   > RigVMVariableNode(Value=@n15.Item) {variable=Upper Lip Control}
   > n16 = &RigUnit_HierarchyAddControlVector2D(
    InitialValue=(X=0,Y=0)
    Name=Lower Lip
    OffsetSpace=LocalSpace
    OffsetTransform=(Rotation=(X=0,Y=0,Z=0.7071,W=0.7071),Translation=(X=0,Y=0,Z=-6),Scale3D=(X=2,Y=-2,Z=2))
    Settings="(PrimaryAxis=X,Limits=(LimitX=(bMinimum=True,bMaximum=True),LimitY=(bMinimum=True,bMaximum=True),MinValue=(X=0,Y=-1),MaxValue=(X=0,Y=1),bDrawLimits=True),Shape=(bVisible=True,Name=\"Circle_Solid\",Color=(R=1,G=1,B=0,A=1),Transform=(Rotation=(X=-0.5,Y=-0.5,Z= ...(truncated)"
    Parent=RigVMVariableNode() {variable=Face Rig Control}.Value
   ) {template="SpawnControl::Execute(in InitialValue,in Settings,in OffsetTransform,in OffsetSpace,in Parent,in Name,out Item)"}
   > RigVMVariableNode(Value=@n16.Item) {variable=Lower Lip Control}
   n17 = &RigUnit_HierarchyAddControlVector2D(
    InitialValue=(X=0,Y=0)
    Name=L Blink
    OffsetSpace=LocalSpace
    OffsetTransform=(Rotation=(X=0,Y=0,Z=0,W=1),Translation=(X=8,Y=0,Z=7),Scale3D=(X=3,Y=3,Z=3))
    Settings="(PrimaryAxis=X,Limits=(LimitX=(bMinimum=True,bMaximum=True),LimitY=(bMinimum=True,bMaximum=True),MinValue=(X=0,Y=-1),MaxValue=(X=0,Y=1),bDrawLimits=True),Shape=(bVisible=True,Name=\"Circle_Solid\",Color=(R=0,G=1,B=0,A=1),Transform=(Rotation=(X=-0.7071,Y=0,Z= ...(truncated)"
    Parent=RigVMVariableNode() {variable=Face Rig Control}.Value
    ExecutePin=@RigVMFunction_Sequence.E
   ) {template="SpawnControl::Execute(in InitialValue,in Settings,in OffsetTransform,in OffsetSpace,in Parent,in Name,out Item)"}
   > RigVMVariableNode(Value=@n17.Item) {variable=L Blink Control}
   > n18 = &RigUnit_HierarchyAddControlVector2D(
    InitialValue=(X=0,Y=0)
    Name=R Blink
    OffsetSpace=LocalSpace
    OffsetTransform=(Rotation=(X=0,Y=0,Z=0,W=1),Translation=(X=-8,Y=0,Z=7),Scale3D=(X=3,Y=3,Z=3))
    Settings="(PrimaryAxis=X,Limits=(LimitX=(bMinimum=True,bMaximum=True),LimitY=(bMinimum=True,bMaximum=True),MinValue=(X=0,Y=-1),MaxValue=(X=0,Y=1),bDrawLimits=True),Shape=(bVisible=True,Name=\"Circle_Solid\",Color=(R=1,G=0,B=0,A=1),Transform=(Rotation=(X=-0.7071,Y=0,Z= ...(truncated)"
    Parent=RigVMVariableNode() {variable=Face Rig Control}.Value
   ) {template="SpawnControl::Execute(in InitialValue,in Settings,in OffsetTransform,in OffsetSpace,in Parent,in Name,out Item)"}
   > RigVMVariableNode(Value=@n18.Item) {variable=R Blink Control}
   > n19 = &RigUnit_HierarchyAddControlVector2D(
    InitialValue=(X=0,Y=0)
    Name=R Low Lid
    OffsetSpace=LocalSpace
    OffsetTransform=(Rotation=(X=0,Y=0,Z=0.7071,W=0.7071),Translation=(X=-10,Y=0,Z=9),Scale3D=(X=3,Y=3,Z=3))
    Settings="(PrimaryAxis=X,Limits=(LimitX=(bMinimum=True,bMaximum=True),LimitY=(bMinimum=True,bMaximum=True),MinValue=(X=0,Y=0),MaxValue=(X=0,Y=1),bDrawLimits=True),Shape=(bVisible=True,Name=\"Circle_Solid\",Color=(R=1,G=0,B=0,A=1),Transform=(Rotation=(X=-0.5,Y=-0.5,Z=0 ...(truncated)"
    Parent=RigVMVariableNode() {variable=Face Rig Control}.Value
   ) {template="SpawnControl::Execute(in InitialValue,in Settings,in OffsetTransform,in OffsetSpace,in Parent,in Name,out Item)"}
   > RigVMVariableNode(Value=@n19.Item) {variable=R Low Lid}
   > n20 = &RigUnit_HierarchyAddControlVector2D(
    InitialValue=(X=0,Y=0)
    Name=L Low Lid
    OffsetSpace=LocalSpace
    OffsetTransform=(Rotation=(X=0,Y=0,Z=0.7071,W=0.7071),Translation=(X=10,Y=0,Z=9),Scale3D=(X=3,Y=3,Z=3))
    Settings="(PrimaryAxis=X,Limits=(LimitX=(bMinimum=True,bMaximum=True),LimitY=(bMinimum=True,bMaximum=True),MinValue=(X=0,Y=0),MaxValue=(X=0,Y=1),bDrawLimits=True),Shape=(bVisible=True,Name=\"Circle_Solid\",Color=(R=0,G=1,B=0,A=1),Transform=(Rotation=(X=-0.5,Y=-0.5,Z=0 ...(truncated)"
    Parent=RigVMVariableNode() {variable=Face Rig Control}.Value
   ) {template="SpawnControl::Execute(in InitialValue,in Settings,in OffsetTransform,in OffsetSpace,in Parent,in Name,out Item)"}
   > RigVMVariableNode(Value=@n20.Item) {variable=L Low Lid}
   > n21 = &RigUnit_HierarchyAddControlVector2D(
    InitialValue=(X=0,Y=0)
    Name=R Top Lid
    OffsetSpace=LocalSpace
    OffsetTransform=(Rotation=(X=0.7071,Y=-0.7071,Z=0,W=0),Translation=(X=-12,Y=0,Z=11.5),Scale3D=(X=3,Y=3,Z=3))
    Settings="(PrimaryAxis=X,Limits=(LimitX=(bMinimum=True,bMaximum=True),LimitY=(bMinimum=True,bMaximum=True),MinValue=(X=0,Y=0),MaxValue=(X=0,Y=1),bDrawLimits=True),Shape=(bVisible=True,Name=\"Circle_Solid\",Color=(R=1,G=0,B=0,A=1),Transform=(Rotation=(X=-0.5,Y=-0.5,Z=0 ...(truncated)"
    Parent=RigVMVariableNode() {variable=Face Rig Control}.Value
   ) {template="SpawnControl::Execute(in InitialValue,in Settings,in OffsetTransform,in OffsetSpace,in Parent,in Name,out Item)"}
   > RigVMVariableNode(Value=@n21.Item) {variable=R Top Lid}
   > n22 = &RigUnit_HierarchyAddControlVector2D(
    InitialValue=(X=0,Y=0)
    Name=L Top Lid
    OffsetSpace=LocalSpace
    OffsetTransform=(Rotation=(X=0.7071,Y=-0.7071,Z=0,W=0),Translation=(X=12,Y=0,Z=11.5),Scale3D=(X=3,Y=3,Z=3))
    Settings="(PrimaryAxis=X,Limits=(LimitX=(bMinimum=True,bMaximum=True),LimitY=(bMinimum=True,bMaximum=True),MinValue=(X=0,Y=0),MaxValue=(X=0,Y=1),bDrawLimits=True),Shape=(bVisible=True,Name=\"Circle_Solid\",Color=(R=0,G=1,B=0,A=1),Transform=(Rotation=(X=-0.5,Y=-0.5,Z=0 ...(truncated)"
    Parent=RigVMVariableNode() {variable=Face Rig Control}.Value
   ) {template="SpawnControl::Execute(in InitialValue,in Settings,in OffsetTransform,in OffsetSpace,in Parent,in Name,out Item)"}
   > RigVMVariableNode(Value=@n22.Item) {variable=L Top Lid}
   > n23 = &RigUnit_HierarchyAddControlVector2D(
    InitialValue=(X=0,Y=0)
    Name=L Eye
    OffsetSpace=LocalSpace
    OffsetTransform=(Rotation=(X=0,Y=0,Z=0.7071,W=0.7071),Translation=(X=4,Y=0,Z=7),Scale3D=(X=3,Y=3,Z=3))
    Settings="(PrimaryAxis=X,Limits=(LimitX=(bMinimum=True,bMaximum=True),LimitY=(bMinimum=True,bMaximum=True),MinValue=(X=-1,Y=-1),MaxValue=(X=1,Y=1),bDrawLimits=True),Shape=(bVisible=True,Name=\"Circle_Solid\",Color=(R=1,G=1,B=0,A=1),Transform=(Rotation=(X=-0.5,Y=-0.5,Z ...(truncated)"
    Parent=RigVMVariableNode() {variable=Face Rig Control}.Value
   ) {template="SpawnControl::Execute(in InitialValue,in Settings,in OffsetTransform,in OffsetSpace,in Parent,in Name,out Item)"}
   > RigVMVariableNode(Value=@n23.Item) {variable=L Eye Control}
   > n24 = &RigUnit_HierarchyAddControlVector2D(
    InitialValue=(X=0,Y=0)
    Name="R  Eye"
    OffsetSpace=LocalSpace
    OffsetTransform=(Rotation=(X=0,Y=0,Z=0.7071,W=0.7071),Translation=(X=-3,Y=0,Z=7),Scale3D=(X=3,Y=3,Z=3))
    Settings="(PrimaryAxis=X,Limits=(LimitX=(bMinimum=True,bMaximum=True),LimitY=(bMinimum=True,bMaximum=True),MinValue=(X=-1,Y=-1),MaxValue=(X=1,Y=1),bDrawLimits=True),Shape=(bVisible=True,Name=\"Circle_Solid\",Color=(R=1,G=1,B=0,A=1),Transform=(Rotation=(X=-0.5,Y=-0.5,Z ...(truncated)"
    Parent=RigVMVariableNode() {variable=Face Rig Control}.Value
   ) {template="SpawnControl::Execute(in InitialValue,in Settings,in OffsetTransform,in OffsetSpace,in Parent,in Name,out Item)"}
   > RigVMVariableNode(Value=@n24.Item) {variable=R Eye Control}
   n25 = &RigUnit_HierarchyAddControlVector2D(
    InitialValue=(X=0,Y=0)
    Name=L Cheek
    OffsetSpace=LocalSpace
    OffsetTransform=(Rotation=(X=0.1228,Y=0.1228,Z=0.6964,W=0.6964),Translation=(X=10,Y=0,Z=4),Scale3D=(X=4,Y=4,Z=4))
    Settings="(PrimaryAxis=X,Limits=(LimitX=(bMinimum=True,bMaximum=True),LimitY=(bMinimum=True,bMaximum=True),MinValue=(X=0,Y=0),MaxValue=(X=0,Y=1),bDrawLimits=true),Shape=(bVisible=True,Name=\"Circle_Solid\",Color=(R=0,G=1,B=0,A=1),Transform=(Rotation=(X=-0.5,Y=-0.5,Z=0 ...(truncated)"
    Parent=RigVMVariableNode() {variable=Face Rig Control}.Value
    ExecutePin=@RigVMFunction_Sequence.F
   ) {template="SpawnControl::Execute(in InitialValue,in Settings,in OffsetTransform,in OffsetSpace,in Parent,in Name,out Item)"}
   > RigVMVariableNode(Value=@n25.Item) {variable=L Cheek Control}
   > n26 = &RigUnit_HierarchyAddControlVector2D(
    InitialValue=(X=0,Y=0)
    Name=R Cheek
    OffsetSpace=LocalSpace
    OffsetTransform=(Rotation=(X=-0.1228,Y=-0.1228,Z=0.6964,W=0.6964),Translation=(X=-10,Y=0,Z=4),Scale3D=(X=4,Y=4,Z=4))
    Settings="(PrimaryAxis=X,Limits=(LimitX=(bMinimum=True,bMaximum=True),LimitY=(bMinimum=True,bMaximum=True),MinValue=(X=0,Y=0),MaxValue=(X=0,Y=1),bDrawLimits=True),Shape=(bVisible=True,Name=\"Circle_Solid\",Color=(R=1,G=0,B=0,A=1),Transform=(Rotation=(X=-0.5,Y=-0.5,Z=0 ...(truncated)"
    Parent=RigVMVariableNode() {variable=Face Rig Control}.Value
   ) {template="SpawnControl::Execute(in InitialValue,in Settings,in OffsetTransform,in OffsetSpace,in Parent,in Name,out Item)"}
   > RigVMVariableNode(Value=@n26.Item) {variable=R Cheek Control}
   > n27 = &RigUnit_HierarchyAddControlVector2D(
    InitialValue=(X=0,Y=0)
    Name=L Brow
    OffsetSpace=LocalSpace
    OffsetTransform=(Rotation=(X=0.2706,Y=0.2706,Z=0.6533,W=0.6533),Translation=(X=7,Y=0,Z=14.5),Scale3D=(X=3,Y=3,Z=3))
    Settings="(PrimaryAxis=X,Limits=(LimitX=(bMinimum=True,bMaximum=True),LimitY=(bMinimum=True,bMaximum=True),MinValue=(X=-1,Y=-1),MaxValue=(X=1,Y=1),bDrawLimits=true),Shape=(bVisible=True,Name=\"Circle_Solid\",Color=(R=0,G=1,B=0,A=1),Transform=(Rotation=(X=-0.5,Y=-0.5,Z ...(truncated)"
    Parent=RigVMVariableNode() {variable=Face Rig Control}.Value
   ) {template="SpawnControl::Execute(in InitialValue,in Settings,in OffsetTransform,in OffsetSpace,in Parent,in Name,out Item)"}
   > RigVMVariableNode(Value=@n27.Item) {variable=L Brow Control}
   > n28 = &RigUnit_HierarchyAddControlVector2D(
    InitialValue=(X=0,Y=0)
    Name=R Brow
    OffsetSpace=LocalSpace
    OffsetTransform=(Rotation=(X=-0.2706,Y=-0.2706,Z=0.6533,W=0.6533),Translation=(X=-7,Y=0,Z=14.5),Scale3D=(X=3,Y=-3,Z=3))
    Settings="(PrimaryAxis=X,Limits=(LimitX=(bMinimum=True,bMaximum=True),LimitY=(bMinimum=True,bMaximum=True),MinValue=(X=-1,Y=-1),MaxValue=(X=1,Y=1),bDrawLimits=true),Shape=(bVisible=True,Name=\"Circle_Solid\",Color=(R=1,G=0,B=0,A=1),Transform=(Rotation=(X=-0.5,Y=-0.5,Z ...(truncated)"
    Parent=RigVMVariableNode() {variable=Face Rig Control}.Value
   ) {template="SpawnControl::Execute(in InitialValue,in Settings,in OffsetTransform,in OffsetSpace,in Parent,in Name,out Item)"}
   > RigVMVariableNode(Value=@n28.Item) {variable=R Brow Control}
   > n29 = &RigUnit_HierarchyAddControlVector2D(
    InitialValue=(X=0,Y=0)
    Name=Squeeze
    OffsetSpace=LocalSpace
    OffsetTransform=(Rotation=(X=0.2706,Y=0.2706,Z=0.6533,W=0.6533),Translation=(X=0,Y=0,Z=15),Scale3D=(X=3,Y=3,Z=3))
    Settings="(PrimaryAxis=X,Limits=(LimitX=(bMinimum=True,bMaximum=True),LimitY=(bMinimum=True,bMaximum=True),MinValue=(X=0,Y=0),MaxValue=(X=1,Y=1),bDrawLimits=True),Shape=(bVisible=True,Name=\"Circle_Solid\",Color=(R=1,G=1,B=0,A=1),Transform=(Rotation=(X=-0.5,Y=-0.5,Z=0 ...(truncated)"
    Parent=RigVMVariableNode() {variable=Face Rig Control}.Value
   ) {template="SpawnControl::Execute(in InitialValue,in Settings,in OffsetTransform,in OffsetSpace,in Parent,in Name,out Item)"}
   > RigVMVariableNode(Value=@n29.Item) {variable=Squeeze Control}
   > n30 = &RigUnit_HierarchyAddControlVector2D(
    InitialValue=(X=0,Y=0)
    Name=Nose
    OffsetSpace=LocalSpace
    OffsetTransform=(Rotation=(X=0,Y=0,Z=0.7071,W=0.7071),Translation=(X=0,Y=0,Z=12),Scale3D=(X=2,Y=2,Z=2))
    Settings="(PrimaryAxis=X,Limits=(LimitX=(bMinimum=True,bMaximum=True),LimitY=(bMinimum=True,bMaximum=True),MinValue=(X=0,Y=0),MaxValue=(X=0,Y=1),bDrawLimits=True),Shape=(bVisible=True,Name=\"Circle_Solid\",Color=(R=1,G=0,B=0,A=1),Transform=(Rotation=(X=-0.5,Y=-0.5,Z=0 ...(truncated)"
    Parent=RigVMVariableNode() {variable=Face Rig Control}.Value
   ) {template="SpawnControl::Execute(in InitialValue,in Settings,in OffsetTransform,in OffsetSpace,in Parent,in Name,out Item)"}
   > n31 = RigVMVariableNode(Value=@n30.Item) {variable=Nose Control}
   RigVMFunction_ControlFlowBranch = &RigVMFunction_ControlFlowBranch(Condition=RigVMVariableNode() {variable=Expose Pupil Controls}.Value, ExecuteContext=@n31.ExecuteContext)
   n32 = &RigUnit_HierarchyAddControlVector2D(
    InitialValue=(X=0,Y=0)
    Name=Pupil Shape
    OffsetSpace=LocalSpace
    OffsetTransform=(Rotation=(X=0.5,Y=-0.5,Z=-0.5,W=0.5),Translation=(X=8,Y=0,Z=2.3),Scale3D=(X=3,Y=3,Z=3.5))
    Settings="(PrimaryAxis=X,Limits=(LimitX=(bMinimum=True,bMaximum=True),LimitY=(bMinimum=True,bMaximum=True),MinValue=(X=0,Y=0),MaxValue=(X=0,Y=1),bDrawLimits=True),Shape=(bVisible=True,Name=\"Circle_Solid\",Color=(R=0,G=1,B=1,A=1),Transform=(Rotation=(X=-0.5,Y=-0.5,Z=0 ...(truncated)"
    ExecutePin=@RigVMFunction_ControlFlowBranch.True
    Parent=RigVMVariableNode() {variable=Face Rig Control}.Value
   ) {template="SpawnControl::Execute(in InitialValue,in Settings,in OffsetTransform,in OffsetSpace,in Parent,in Name,out Item)"}
   > RigVMVariableNode(Value=@n32.Item) {variable=Pupil Shape Control}
   > n33 = &RigUnit_HierarchyAddControlVector2D(
    InitialValue=(X=0,Y=0)
    Name=Pupil Size
    OffsetSpace=LocalSpace
    OffsetTransform=(Rotation=(X=0.5,Y=-0.5,Z=-0.5,W=0.5),Translation=(X=8,Y=0,Z=0.5),Scale3D=(X=3,Y=3,Z=3.5))
    Settings="(PrimaryAxis=X,Limits=(LimitX=(bMinimum=True,bMaximum=True),LimitY=(bMinimum=True,bMaximum=True),MinValue=(X=0,Y=0),MaxValue=(X=0,Y=1),bDrawLimits=True),Shape=(bVisible=True,Name=\"Circle_Solid\",Color=(R=0,G=1,B=1,A=1),Transform=(Rotation=(X=-0.5,Y=-0.5,Z=0 ...(truncated)"
    Parent=RigVMVariableNode() {variable=Face Rig Control}.Value
   ) {template="SpawnControl::Execute(in InitialValue,in Settings,in OffsetTransform,in OffsetSpace,in Parent,in Name,out Item)"}
   > RigVMVariableNode(Value=@n33.Item) {variable=Pupil Size Control}
  ]}
 New Function_ContainedGraph {nodes [
   n1 = RigVMFunctionEntryNode()
   RigVMFunction_Sequence = &RigVMFunction_Sequence(ExecuteContext=@n1.ExecuteContext)
   Subtract = &RigVMFunction_MathDoubleSub(
    A=/Script/ControlRig.RigUnit_GetCurveValue(Curve=@n1.X_Pos_Curve.Name) {is=struct RigVMUnitNode}.Value
    B=&RigUnit_GetCurveValue(Curve=@n1.X_Neg_Curve.Name).Value
   )
   /Script/ControlRig.RigUnit_SetControlVector2D(
    Vector=(X=0,Y=0)
    Weight=1
    Control=@n1.Control.Name
    ExecutePin=@RigVMFunction_Sequence.A
    Vector.X=@Subtract.Result
    Vector.Y=&RigVMFunction_MathDoubleSub(
     A=&RigUnit_GetCurveValue(Curve=@n1.Y_Pos_Curve.Name).Value
     B=&RigUnit_GetCurveValue(Curve=@n1.Y_Neg_Curve.Name).Value
    ).Result
   ) {is=struct RigVMUnitNode, template="SetControlValue::Execute(in Control,in Weight,io Vector)"}
   RigVMFunctionReturnNode(ExecuteContext=@RigVMFunction_Sequence.B)
  ]}
 New Function_ContainedGraph {nodes [
   n1 = RigVMFunctionEntryNode()
   RigVMFunction_Sequence = &RigVMFunction_Sequence(ExecuteContext=@n1.ExecuteContext)
   n2 = /Script/ControlRig.RigUnit_GetControlVector2D(Control=@n1.Control.Name) {is=struct RigVMUnitNode, template="GetControlValue::Execute(in Control,out Vector,out Minimum,out Maximum)"}
   Remap = &RigVMFunction_MathDoubleRemap(SourceMaximum=1, SourceMinimum=0, TargetMaximum=1, TargetMinimum=0, Value=@n2.Vector.X)
   /Script/ControlRig.RigUnit_SetCurveValue(Value=@Remap.Result, ExecutePin=@RigVMFunction_Sequence.A, Curve=RigVMVariableNode() {variable=X_Pos_Curve}.Value.Name) {is=struct RigVMUnitNode}
   > &RigUnit_SetCurveValue(
    Value=&RigVMFunction_MathDoubleRemap(SourceMaximum=-1, SourceMinimum=0, TargetMaximum=1, TargetMinimum=0, Value=@n2.Vector.X).Result
    Curve=RigVMVariableNode() {variable=X_Neg_Curve}.Value.Name
   )
   > &RigUnit_SetCurveValue(
    Value=&RigVMFunction_MathDoubleRemap(SourceMaximum=1, SourceMinimum=0, TargetMaximum=1, TargetMinimum=0, Value=@n2.Vector.Y).Result
    Curve=RigVMVariableNode() {variable=Y_Pos_Curve}.Value.Name
   )
   > &RigUnit_SetCurveValue(
    Value=&RigVMFunction_MathDoubleRemap(SourceMaximum=-1, SourceMinimum=0, TargetMaximum=1, TargetMinimum=0, Value=@n2.Vector.Y).Result
    Curve=RigVMVariableNode() {variable=Y_Neg_Curve}.Value.Name
   )
   RigVMFunctionReturnNode(ExecuteContext=@RigVMFunction_Sequence.B)
  ]}
 RigVMFunctionLibrary {nodes [
   Curve to Vector Control = RigVMCollapseNode(
    Control=(Type=None,Name="None")
    ExecuteContext=()
    X_Neg_Curve=(Type=None,Name="None")
    X_Pos_Curve=(Type=None,Name="None")
    Y_Neg_Curve=(Type=None,Name="None")
    Y_Pos_Curve=(Type=None,Name="None")
   )
   Vector Control To Curve = RigVMCollapseNode(
    Control=(Type=Control,Name="None")
    ExecuteContext=()
    X_Neg_Curve=(Type=Curve,Name="None")
    X_Pos_Curve=(Type=Curve,Name="None")
    Y_Neg_Curve=(Type=Curve,Name="None")
    Y_Pos_Curve=(Type=Curve,Name="None")
   )
  ]}
 RigVMFunction_Sequence_1_ContainedGraph {nodes [
   n1 = RigVMFunctionEntryNode()
   n2 = &RigVMFunction_Sequence(ExecuteContext=@n1.ExecuteContext)
   n3 = &RigVMFunction_Sequence(ExecuteContext=@n2.B)
   n4 = &RigVMFunction_Sequence(ExecuteContext=@n3.B)
   n5 = &RigVMFunction_Sequence(ExecuteContext=@n4.B)
   RigVMFunctionReturnNode(B=@n2.A, C=@n3.A, D=@n4.A, E=@n5.A, F=@n5.B)
  ]}
 RigVMFunction_Sequence_2_ContainedGraph {nodes [
   n1 = RigVMFunctionEntryNode()
   n2 = &RigVMFunction_Sequence(ExecuteContext=@n1.ExecuteContext)
   n3 = &RigVMFunction_Sequence(ExecuteContext=@n2.B)
   n4 = &RigVMFunction_Sequence(ExecuteContext=@n3.B)
   n5 = &RigVMFunction_Sequence(ExecuteContext=@n4.B)
   RigVMFunctionReturnNode(A=@n2.A, B=@n3.A, C=@n4.A, D=@n5.A, E=@n5.B)
  ]}
 RigVMFunction_Sequence_2_ContainedGraph {nodes [
   n1 = RigVMFunctionEntryNode()
   n2 = &RigVMFunction_Sequence(ExecuteContext=@n1.ExecuteContext)
   n3 = &RigVMFunction_Sequence(ExecuteContext=@n2.B)
   n4 = &RigVMFunction_Sequence(ExecuteContext=@n3.B)
   n5 = &RigVMFunction_Sequence(ExecuteContext=@n4.B)
   RigVMFunctionReturnNode(A=@n2.A, B=@n3.A, C=@n4.A, D=@n5.A, E=@n5.B)
  ]}
 RigVMModel {nodes [
   RigUnit_BeginExecution = &RigUnit_BeginExecution()
   RigVMFunction_Sequence_1 = RigVMAggregateNode(ExecuteContext=@RigUnit_BeginExecution.ExecutePin)
   Minimum = /Script/RigVM.RigVMFunction_MathDoubleMin(A=&RigUnit_GetCurveValue(Curve=L_smile_pose).Value, B=&RigUnit_GetCurveValue(Curve=jaw_open_pose).Value) {is=struct RigVMUnitNode, template="Minimum::Execute(in A,in B,out Result)"}
   &RigUnit_SetCurveValue(Curve=L_smile_open_correction_pose, Value=@Minimum.Result, ExecutePin=@RigVMFunction_Sequence_1.E)
   > n1 = &RigUnit_SetCurveValue(
    Curve=R_smile_open_correction_pose
    Value=&RigVMFunction_MathDoubleMin(A=&RigUnit_GetCurveValue(Curve=R_smile_pose).Value, B=&RigUnit_GetCurveValue(Curve=jaw_open_pose).Value) {template="Minimum::Execute(in A,in B,out Result)"}.Result
   )
   RigVMFunction_ControlFlowBranch = &RigVMFunction_ControlFlowBranch(Condition=RigVMVariableNode() {variable=Expose Pupil Controls}.Value, ExecuteContext=@n1.ExecutePin)
   &RigUnit_SetCurveValue(
    Curve=pupil_shape
    Value=&RigUnit_GetControlVector2D(Control=RigVMVariableNode() {variable=Pupil Shape Control}.Value.Name) {template="GetControlValue::Execute(in Control,out Vector,out Minimum,out Maximum)"}.Vector.Y
    ExecutePin=@RigVMFunction_ControlFlowBranch.True
   )
   > &RigUnit_SetCurveValue(
    Curve=pupil_size
    Value=&RigUnit_GetControlVector2D(Control=RigVMVariableNode() {variable=Pupil Size Control}.Value.Name) {template="GetControlValue::Execute(in Control,out Vector,out Minimum,out Maximum)"}.Vector.Y
   )
   &RigUnit_SetCurveValue(
    Curve=reverse_jaw_pose
    Value=&RigUnit_GetControlVector2D(Control=RigVMVariableNode() {variable=Reverse Jaw Control}.Value.Name) {template="GetControlValue::Execute(in Control,out Vector,out Minimum,out Maximum)"}.Vector.Y
    ExecutePin=@RigVMFunction_Sequence_1.C
   )
   > Vector Control To Curve_7 = RigVMFunctionReferenceNode(
    X_Neg_Curve=(Type=Curve,Name="tongue_right_pose")
    X_Pos_Curve=(Type=Curve,Name="tongue_left_pose")
    Y_Neg_Curve=(Type=Curve,Name="tongue_up_pose")
    Y_Pos_Curve=(Type=Curve,Name="tongue_down_pose")
    Control=RigVMVariableNode() {variable=Tongue Control}.Value
   )
   > Vector Control To Curve_15 = RigVMFunctionReferenceNode(
    X_Neg_Curve=(Type=Curve,Name="tongue_back_pose")
    X_Pos_Curve=(Type=Curve,Name="tongue_forward_pose")
    Y_Neg_Curve=(Type=Curve,Name="None")
    Y_Pos_Curve=(Type=Curve,Name="None")
    Control=RigVMVariableNode() {variable=Tongue Fwd Control}.Value
   )
   > Vector Control To Curve_18 = RigVMFunctionReferenceNode(
    X_Neg_Curve=(Type=Curve,Name="R_tongue_twist_pose")
    X_Pos_Curve=(Type=Curve,Name="L_tongue_twist_pose")
    Y_Neg_Curve=(Type=Curve,Name="None")
    Y_Pos_Curve=(Type=Curve,Name="None")
    Control=RigVMVariableNode() {variable=Tongue Twist Control}.Value
   )
   > Vector Control To Curve_19 = RigVMFunctionReferenceNode(
    X_Neg_Curve=(Type=Curve,Name="tongue_curl_tip_down_pose")
    X_Pos_Curve=(Type=Curve,Name="tongue_curl_tip_up_pose")
    Y_Neg_Curve=(Type=Curve,Name="None")
    Y_Pos_Curve=(Type=Curve,Name="None")
    Control=RigVMVariableNode() {variable=Tongue Curl Tip Control}.Value
   )
   > Vector Control To Curve_20 = RigVMFunctionReferenceNode(
    X_Neg_Curve=(Type=Curve,Name="tongue_curl_mid_down_pose")
    X_Pos_Curve=(Type=Curve,Name="tongue_curl_mid_up_pose")
    Y_Neg_Curve=(Type=Curve,Name="None")
    Y_Pos_Curve=(Type=Curve,Name="None")
    Control=RigVMVariableNode() {variable=Tongue Curl Control}.Value
   )
   > &RigUnit_SetCurveValue(
    Curve=L_cheek_up_pose
    Value=&RigUnit_GetControlVector2D(Control=RigVMVariableNode() {variable=L Cheek Control}.Value.Name) {template="GetControlValue::Execute(in Control,out Vector,out Minimum,out Maximum)"}.Vector.Y
   )
   > &RigUnit_SetCurveValue(
    Curve=R_cheek_up_pose
    Value=&RigUnit_GetControlVector2D(Control=RigVMVariableNode() {variable=R Cheek Control}.Value.Name) {template="GetControlValue::Execute(in Control,out Vector,out Minimum,out Maximum)"}.Vector.Y
   )
   > &RigUnit_SetCurveValue(
    Curve=C_nose_wrinkler_pose
    Value=&RigUnit_GetControlVector2D(Control=RigVMVariableNode() {variable=Nose Control}.Value.Name) {template="GetControlValue::Execute(in Control,out Vector,out Minimum,out Maximum)"}.Vector.Y
   )
   Vector Control To Curve_3 = RigVMFunctionReferenceNode(
    X_Neg_Curve=(Type=Curve,Name="None")
    X_Pos_Curve=(Type=Curve,Name="None")
    Y_Neg_Curve=(Type=Curve,Name="L_upper_lip_lower_pose")
    Y_Pos_Curve=(Type=Curve,Name="L_upper_lip_raiser_pose")
    Control=RigVMVariableNode() {variable=L Upper Lip Control}.Value
    ExecuteContext=@RigVMFunction_Sequence_1.B
   )
   > Vector Control To Curve_4 = RigVMFunctionReferenceNode(
    X_Neg_Curve=(Type=None,Name="None")
    X_Pos_Curve=(Type=None,Name="None")
    Y_Neg_Curve=(Type=Curve,Name="R_upper_lip_lower_pose")
    Y_Pos_Curve=(Type=Curve,Name="R_upper_lip_raiser_pose")
    Control=RigVMVariableNode() {variable=R Upper Lip Control}.Value
   )
   > Vector Control To Curve_5 = RigVMFunctionReferenceNode(
    X_Neg_Curve=(Type=None,Name="None")
    X_Pos_Curve=(Type=None,Name="None")
    Y_Neg_Curve=(Type=Curve,Name="L_lower_lip_down_pose")
    Y_Pos_Curve=(Type=Curve,Name="L_lower_lip_up_pose")
    Control=RigVMVariableNode() {variable=L Lower Lip Control}.Value
   )
   > Vector Control To Curve_6 = RigVMFunctionReferenceNode(
    X_Neg_Curve=(Type=None,Name="None")
    X_Pos_Curve=(Type=None,Name="None")
    Y_Neg_Curve=(Type=Curve,Name="R_lower_lip_down_pose")
    Y_Pos_Curve=(Type=Curve,Name="R_lower_lip_up_pose")
    Control=RigVMVariableNode() {variable=R Lower Lip Control}.Value
   )
   > Vector Control To Curve_8 = RigVMFunctionReferenceNode(
    X_Neg_Curve=(Type=Curve,Name="None")
    X_Pos_Curve=(Type=Curve,Name="None")
    Y_Neg_Curve=(Type=Curve,Name="upper_lip_dn_pose")
    Y_Pos_Curve=(Type=Curve,Name="upper_lip_up_pose")
    Control=RigVMVariableNode() {variable=Upper Lip Control}.Value
   )
   > Vector Control To Curve_9 = RigVMFunctionReferenceNode(
    X_Neg_Curve=(Type=Curve,Name="None")
    X_Pos_Curve=(Type=Curve,Name="None")
    Y_Neg_Curve=(Type=Curve,Name="lower_lip_dn_pose")
    Y_Pos_Curve=(Type=Curve,Name="lower_lip_up_pose")
    Control=RigVMVariableNode() {variable=Lower Lip Control}.Value
   )
   Vector Control To Curve = RigVMFunctionReferenceNode(
    X_Neg_Curve=(Type=Curve,Name="jaw_right_pose")
    X_Pos_Curve=(Type=Curve,Name="jaw_left_pose")
    Y_Neg_Curve=(Type=Curve,Name="jaw_up_pose")
    Y_Pos_Curve=(Type=Curve,Name="jaw_open_pose")
    Control=RigVMVariableNode() {variable=Jaw Main Control}.Value
    ExecuteContext=@RigVMFunction_Sequence_1.A
   )
   > Vector Control To Curve_1 = RigVMFunctionReferenceNode(
    X_Neg_Curve=(Type=Curve,Name="L_lip_corner_wide_pose")
    X_Pos_Curve=(Type=Curve,Name="L_lip_corner_narrow_pose")
    Y_Neg_Curve=(Type=Curve,Name="L_frown_pose")
    Y_Pos_Curve=(Type=Curve,Name="L_smile_pose")
    Control=RigVMVariableNode() {variable=L Corner Control}.Value
   )
   > Vector Control To Curve_2 = RigVMFunctionReferenceNode(
    X_Neg_Curve=(Type=Curve,Name="R_lip_corner_wide_pose")
    X_Pos_Curve=(Type=Curve,Name="R_lip_corner_narrow_pose")
    Y_Neg_Curve=(Type=Curve,Name="R_frown_pose")
    Y_Pos_Curve=(Type=Curve,Name="R_smile_pose")
    Control=RigVMVariableNode() {variable=R Corner Control}.Value
   )
   Vector Control To Curve_10 = RigVMFunctionReferenceNode(
    Control=(Type=Control,Name="None")
    X_Neg_Curve=(Type=Curve,Name="L_brow_In_Dn_pose")
    X_Pos_Curve=(Type=Curve,Name="L_brow_In_Up_pose")
    Y_Neg_Curve=(Type=Curve,Name="L_brow_Ot_Dn_pose")
    Y_Pos_Curve=(Type=Curve,Name="L_brow_Ot_Up_pose")
    Control.Name=RigVMVariableNode() {variable=L Brow Control}.Value.Name
    ExecuteContext=@RigVMFunction_Sequence_1.D
   )
   > Vector Control To Curve_11 = RigVMFunctionReferenceNode(
    Control=(Type=Control,Name="None")
    X_Neg_Curve=(Type=Curve,Name="R_brow_In_Dn_pose")
    X_Pos_Curve=(Type=Curve,Name="R_brow_In_Up_pose")
    Y_Neg_Curve=(Type=Curve,Name="R_brow_Ot_Dn_pose")
    Y_Pos_Curve=(Type=Curve,Name="R_brow_Ot_Up_pose")
    Control.Name=RigVMVariableNode() {variable=R Brow Control}.Value.Name
   )
   > Vector Control To Curve_14 = RigVMFunctionReferenceNode(
    X_Neg_Curve=(Type=Curve,Name="None")
    X_Pos_Curve=(Type=Curve,Name="R_brow_squeeze_in_pose")
    Y_Neg_Curve=(Type=Curve,Name="None")
    Y_Pos_Curve=(Type=Curve,Name="L_brow_squeeze_in_pose")
    Control=RigVMVariableNode() {variable=Squeeze Control}.Value
   )
   > Vector Control To Curve_16 = RigVMFunctionReferenceNode(
    X_Neg_Curve=(Type=Curve,Name="L_eye_look_left_pose")
    X_Pos_Curve=(Type=Curve,Name="L_eye_look_right_pose")
    Y_Neg_Curve=(Type=Curve,Name="L_eye_look_down_pose")
    Y_Pos_Curve=(Type=Curve,Name="L_eye_look_up_pose")
    Control=RigVMVariableNode() {variable=L Eye Control}.Value
   )
   > Vector Control To Curve_17 = RigVMFunctionReferenceNode(
    X_Neg_Curve=(Type=Curve,Name="R_eye_look_left_pose")
    X_Pos_Curve=(Type=Curve,Name="R_eye_look_right_pose")
    Y_Neg_Curve=(Type=Curve,Name="R_eye_look_down_pose")
    Y_Pos_Curve=(Type=Curve,Name="R_eye_look_up_pose")
    Control=RigVMVariableNode() {variable=R Eye Control}.Value
   )
   > &RigUnit_SetCurveValue(
    Curve=L_lower_lid_up_pose
    Value=&RigUnit_GetControlVector2D(Control=RigVMVariableNode() {variable=L Low Lid}.Value.Name) {template="GetControlValue::Execute(in Control,out Vector,out Minimum,out Maximum)"}.Vector.Y
   )
   > &RigUnit_SetCurveValue(
    Curve=R_lower_lid_up_pose
    Value=&RigUnit_GetControlVector2D(Control=RigVMVariableNode() {variable=R Low Lid}.Value.Name) {template="GetControlValue::Execute(in Control,out Vector,out Minimum,out Maximum)"}.Vector.Y
   )
   > &RigUnit_SetCurveValue(
    Curve=L_upperLidDown_pose
    Value=&RigUnit_GetControlVector2D(Control=RigVMVariableNode() {variable=L Top Lid}.Value.Name) {template="GetControlValue::Execute(in Control,out Vector,out Minimum,out Maximum)"}.Vector.Y
   )
   > &RigUnit_SetCurveValue(
    Curve=R_upperLidDown_pose
    Value=&RigUnit_GetControlVector2D(Control=RigVMVariableNode() {variable=R Top Lid}.Value.Name) {template="GetControlValue::Execute(in Control,out Vector,out Minimum,out Maximum)"}.Vector.Y
   )
   > Vector Control To Curve_12 = RigVMFunctionReferenceNode(
    Control=(Type=Control,Name="None")
    X_Neg_Curve=(Type=Curve,Name="None")
    X_Pos_Curve=(Type=Curve,Name="None")
    Y_Neg_Curve=(Type=Curve,Name="L_blink_pose")
    Y_Pos_Curve=(Type=Curve,Name="L_wide_pose")
    Control.Name=RigVMVariableNode() {variable=L Blink Control}.Value.Name
   )
   > Vector Control To Curve_13 = RigVMFunctionReferenceNode(
    Control=(Type=Control,Name="None")
    X_Neg_Curve=(Type=Curve,Name="None")
    X_Pos_Curve=(Type=Curve,Name="None")
    Y_Neg_Curve=(Type=Curve,Name="R_blink_pose")
    Y_Pos_Curve=(Type=Curve,Name="R_wide_pose")
    Control.Name=RigVMVariableNode() {variable=R Blink Control}.Value.Name
   )
  ]}
 RigVMModel Rig {nodes [
   n1 = /Script/ControlRig.RigUnit_InverseExecution() {is=struct RigVMUnitNode}
   RigVMFunction_Sequence_1 = RigVMAggregateNode(ExecuteContext=@n1.ExecutePin)
   Curve to Vector Control_14 = RigVMFunctionReferenceNode(
    X_Neg_Curve=(Type=Curve,Name="None")
    X_Pos_Curve=(Type=Curve,Name="None")
    Y_Neg_Curve=(Type=Curve,Name="L_upper_lip_lower_pose")
    Y_Pos_Curve=(Type=Curve,Name="L_upper_lip_raiser_pose")
    Control=RigVMVariableNode() {variable=L Upper Lip Control}.Value
    ExecuteContext=@RigVMFunction_Sequence_1.B
   )
   > Curve to Vector Control_15 = RigVMFunctionReferenceNode(
    X_Neg_Curve=(Type=Curve,Name="None")
    X_Pos_Curve=(Type=Curve,Name="None")
    Y_Neg_Curve=(Type=Curve,Name="R_upper_lip_lower_pose")
    Y_Pos_Curve=(Type=Curve,Name="R_upper_lip_raiser_pose")
    Control=RigVMVariableNode() {variable=R Upper Lip Control}.Value
   )
   > Curve to Vector Control_16 = RigVMFunctionReferenceNode(
    X_Neg_Curve=(Type=Curve,Name="None")
    X_Pos_Curve=(Type=Curve,Name="None")
    Y_Neg_Curve=(Type=Curve,Name="L_lower_lip_down_pose")
    Y_Pos_Curve=(Type=Curve,Name="L_lower_lip_up_pose")
    Control=RigVMVariableNode() {variable=L Lower Lip Control}.Value
   )
   > Curve to Vector Control_17 = RigVMFunctionReferenceNode(
    X_Neg_Curve=(Type=Curve,Name="None")
    X_Pos_Curve=(Type=Curve,Name="None")
    Y_Neg_Curve=(Type=Curve,Name="R_lower_lip_down_pose")
    Y_Pos_Curve=(Type=Curve,Name="R_lower_lip_up_pose")
    Control=RigVMVariableNode() {variable=R Lower Lip Control}.Value
   )
   > Curve to Vector Control_18 = RigVMFunctionReferenceNode(
    X_Neg_Curve=(Type=Curve,Name="None")
    X_Pos_Curve=(Type=Curve,Name="None")
    Y_Neg_Curve=(Type=Curve,Name="upper_lip_dn_pose")
    Y_Pos_Curve=(Type=Curve,Name="upper_lip_up_pose")
    Control=RigVMVariableNode() {variable=Upper Lip Control}.Value
   )
   > Curve to Vector Control_19 = RigVMFunctionReferenceNode(
    X_Neg_Curve=(Type=Curve,Name="None")
    X_Pos_Curve=(Type=Curve,Name="None")
    Y_Neg_Curve=(Type=Curve,Name="lower_lip_dn_pose")
    Y_Pos_Curve=(Type=Curve,Name="lower_lip_up_pose")
    Control=RigVMVariableNode() {variable=Lower Lip Control}.Value
   )
   Curve to Vector Control_6 = RigVMFunctionReferenceNode(
    X_Neg_Curve=(Type=Curve,Name="L_brow_In_Dn_pose")
    X_Pos_Curve=(Type=Curve,Name="L_brow_In_Up_pose")
    Y_Neg_Curve=(Type=Curve,Name="L_brow_Ot_Dn_pose")
    Y_Pos_Curve=(Type=Curve,Name="L_brow_Ot_Up_pose")
    Control=RigVMVariableNode() {variable=L Brow Control}.Value
    ExecuteContext=@RigVMFunction_Sequence_1.D
   )
   > Curve to Vector Control_7 = RigVMFunctionReferenceNode(
    X_Neg_Curve=(Type=Curve,Name="R_brow_In_Dn_pose")
    X_Pos_Curve=(Type=Curve,Name="R_brow_In_Up_pose")
    Y_Neg_Curve=(Type=Curve,Name="R_brow_Ot_Dn_pose")
    Y_Pos_Curve=(Type=Curve,Name="R_brow_Ot_Up_pose")
    Control=RigVMVariableNode() {variable=R Brow Control}.Value
   )
   > Curve to Vector Control_8 = RigVMFunctionReferenceNode(
    X_Neg_Curve=(Type=Curve,Name="None")
    X_Pos_Curve=(Type=Curve,Name="R_brow_squeeze_in_pose")
    Y_Neg_Curve=(Type=Curve,Name="None")
    Y_Pos_Curve=(Type=Curve,Name="L_brow_squeeze_in_pose")
    Control=RigVMVariableNode() {variable=Squeeze Control}.Value
   )
   > Curve to Vector Control_9 = RigVMFunctionReferenceNode(
    X_Neg_Curve=(Type=Curve,Name="L_eye_look_left_pose")
    X_Pos_Curve=(Type=Curve,Name="L_eye_look_right_pose")
    Y_Neg_Curve=(Type=Curve,Name="L_eye_look_down_pose")
    Y_Pos_Curve=(Type=Curve,Name="L_eye_look_up_pose")
    Control=RigVMVariableNode() {variable=L Eye Control}.Value
   )
   > Curve to Vector Control_10 = RigVMFunctionReferenceNode(
    X_Neg_Curve=(Type=Curve,Name="R_eye_look_left_pose")
    X_Pos_Curve=(Type=Curve,Name="R_eye_look_right_pose")
    Y_Neg_Curve=(Type=Curve,Name="R_eye_look_down_pose")
    Y_Pos_Curve=(Type=Curve,Name="R_eye_look_up_pose")
    Control=RigVMVariableNode() {variable=R Eye Control}.Value
   )
   > &RigUnit_SetControlVector2D(
    Vector=(X=0,Y=0)
    Weight=1
    Vector.Y=&RigUnit_GetCurveValue(Curve=L_lower_lid_up_pose).Value
    Control=RigVMVariableNode() {variable=L Low Lid}.Value.Name
   ) {template="SetControlValue::Execute(in Control,in Weight,io Vector)"}
   > &RigUnit_SetControlVector2D(
    Vector=(X=0,Y=0)
    Weight=1
    Vector.Y=&RigUnit_GetCurveValue(Curve=R_lower_lid_up_pose).Value
    Control=RigVMVariableNode() {variable=R Low Lid}.Value.Name
   ) {template="SetControlValue::Execute(in Control,in Weight,io Vector)"}
   > Curve to Vector Control_11 = RigVMFunctionReferenceNode(
    Control=(Type=None,Name="None")
    X_Neg_Curve=(Type=Curve,Name="None")
    X_Pos_Curve=(Type=Curve,Name="None")
    Y_Neg_Curve=(Type=Curve,Name="L_blink_pose")
    Y_Pos_Curve=(Type=Curve,Name="L_wide_pose")
    Control.Name=RigVMVariableNode() {variable=L Blink Control}.Value.Name
   )
   > Curve to Vector Control_13 = RigVMFunctionReferenceNode(
    Control=(Type=None,Name="None")
    X_Neg_Curve=(Type=Curve,Name="None")
    X_Pos_Curve=(Type=Curve,Name="None")
    Y_Neg_Curve=(Type=Curve,Name="R_blink_pose")
    Y_Pos_Curve=(Type=Curve,Name="R_wide_pose")
    Control.Name=RigVMVariableNode() {variable=R Blink Control}.Value.Name
   )
   > &RigUnit_SetControlVector2D(
    Vector=(X=0,Y=0)
    Weight=1
    Vector.Y=&RigUnit_GetCurveValue(Curve=L_upperLidDown_pose).Value
    Control=RigVMVariableNode() {variable=L Top Lid}.Value.Name
   ) {template="SetControlValue::Execute(in Control,in Weight,io Vector)"}
   > &RigUnit_SetControlVector2D(
    Vector=(X=0,Y=0)
    Weight=1
    Vector.Y=&RigUnit_GetCurveValue(Curve=R_upperLidDown_pose).Value
    Control=RigVMVariableNode() {variable=R Top Lid}.Value.Name
   ) {template="SetControlValue::Execute(in Control,in Weight,io Vector)"}
   RigVMFunction_ControlFlowBranch = &RigVMFunction_ControlFlowBranch(Condition=RigVMVariableNode() {variable=Expose Pupil Controls}.Value, ExecuteContext=@RigVMFunction_Sequence_1.E)
   &RigUnit_SetControlVector2D(
    Vector=(X=0,Y=0)
    Weight=1
    Vector.Y=&RigUnit_GetCurveValue(Curve=pupil_shape).Value
    ExecutePin=@RigVMFunction_ControlFlowBranch.True
    Control=RigVMVariableNode() {variable=Pupil Shape Control}.Value.Name
   ) {template="SetControlValue::Execute(in Control,in Weight,io Vector)"}
   > &RigUnit_SetControlVector2D(
    Vector=(X=0,Y=0)
    Weight=1
    Vector.Y=&RigUnit_GetCurveValue(Curve=pupil_size).Value
    Control=RigVMVariableNode() {variable=Pupil Size Control}.Value.Name
   ) {template="SetControlValue::Execute(in Control,in Weight,io Vector)"}
   &RigUnit_SetControlVector2D(
    Vector=(X=0,Y=0)
    Weight=1
    Vector.Y=&RigUnit_GetCurveValue(Curve=reverse_jaw_pose).Value
    Control=RigVMVariableNode() {variable=Reverse Jaw Control}.Value.Name
    ExecutePin=@RigVMFunction_Sequence_1.C
   ) {template="SetControlValue::Execute(in Control,in Weight,io Vector)"}
   > Curve to Vector Control_20 = RigVMFunctionReferenceNode(
    X_Neg_Curve=(Type=Curve,Name="tongue_right_pose")
    X_Pos_Curve=(Type=Curve,Name="tongue_left_pose")
    Y_Neg_Curve=(Type=Curve,Name="tongue_up_pose")
    Y_Pos_Curve=(Type=Curve,Name="tongue_down_pose")
    Control=RigVMVariableNode() {variable=Tongue Control}.Value
   )
   > Curve to Vector Control_2 = RigVMFunctionReferenceNode(
    X_Neg_Curve=(Type=Curve,Name="tongue_back_pose")
    X_Pos_Curve=(Type=Curve,Name="tongue_forward_pose")
    Y_Neg_Curve=(Type=Curve,Name="None")
    Y_Pos_Curve=(Type=Curve,Name="None")
    Control=RigVMVariableNode() {variable=Tongue Fwd Control}.Value
   )
   > Curve to Vector Control_3 = RigVMFunctionReferenceNode(
    X_Neg_Curve=(Type=Curve,Name="R_tongue_twist_pose")
    X_Pos_Curve=(Type=Curve,Name="L_tongue_twist_pose")
    Y_Neg_Curve=(Type=Curve,Name="None")
    Y_Pos_Curve=(Type=Curve,Name="None")
    Control=RigVMVariableNode() {variable=Tongue Twist Control}.Value
   )
   > Curve to Vector Control_4 = RigVMFunctionReferenceNode(
    X_Neg_Curve=(Type=Curve,Name="tongue_curl_tip_down_pose")
    X_Pos_Curve=(Type=Curve,Name="tongue_curl_tip_up_pose")
    Y_Neg_Curve=(Type=Curve,Name="None")
    Y_Pos_Curve=(Type=Curve,Name="None")
    Control=RigVMVariableNode() {variable=Tongue Curl Tip Control}.Value
   )
   > Curve to Vector Control_5 = RigVMFunctionReferenceNode(
    X_Neg_Curve=(Type=Curve,Name="tongue_curl_mid_down_pose")
    X_Pos_Curve=(Type=Curve,Name="tongue_curl_mid_up_pose")
    Y_Neg_Curve=(Type=Curve,Name="None")
    Y_Pos_Curve=(Type=Curve,Name="None")
    Control=RigVMVariableNode() {variable=Tongue Curl Control}.Value
   )
   > &RigUnit_SetControlVector2D(
    Vector=(X=0,Y=0)
    Weight=1
    Vector.Y=&RigUnit_GetCurveValue(Curve=L_cheek_up_pose).Value
    Control=RigVMVariableNode() {variable=L Cheek Control}.Value.Name
   ) {template="SetControlValue::Execute(in Control,in Weight,io Vector)"}
   > &RigUnit_SetControlVector2D(
    Vector=(X=0,Y=0)
    Weight=1
    Vector.Y=&RigUnit_GetCurveValue(Curve=R_cheek_up_pose).Value
    Control=RigVMVariableNode() {variable=R Cheek Control}.Value.Name
   ) {template="SetControlValue::Execute(in Control,in Weight,io Vector)"}
   > &RigUnit_SetControlVector2D(
    Vector=(X=0,Y=0)
    Weight=1
    Vector.Y=&RigUnit_GetCurveValue(Curve=C_nose_wrinkler_pose).Value
    Control=RigVMVariableNode() {variable=Nose Control}.Value.Name
   ) {template="SetControlValue::Execute(in Control,in Weight,io Vector)"}
   Curve to Vector Control = RigVMFunctionReferenceNode(
    X_Neg_Curve=(Type=Curve,Name="jaw_right_pose")
    X_Pos_Curve=(Type=Curve,Name="jaw_left_pose")
    Y_Neg_Curve=(Type=Curve,Name="jaw_up_pose")
    Y_Pos_Curve=(Type=Curve,Name="jaw_open_pose")
    Control=RigVMVariableNode() {variable=Jaw Main Control}.Value
    ExecuteContext=@RigVMFunction_Sequence_1.A
   )
   > Curve to Vector Control_1 = RigVMFunctionReferenceNode(
    X_Neg_Curve=(Type=Curve,Name="L_lip_corner_wide_pose")
    X_Pos_Curve=(Type=Curve,Name="L_lip_corner_narrow_pose")
    Y_Neg_Curve=(Type=Curve,Name="L_frown_pose")
    Y_Pos_Curve=(Type=Curve,Name="L_smile_pose")
    Control=RigVMVariableNode() {variable=L Corner Control}.Value
   )
   > Curve to Vector Control_12 = RigVMFunctionReferenceNode(
    X_Neg_Curve=(Type=Curve,Name="R_lip_corner_wide_pose")
    X_Pos_Curve=(Type=Curve,Name="R_lip_corner_narrow_pose")
    Y_Neg_Curve=(Type=Curve,Name="R_frown_pose")
    Y_Pos_Curve=(Type=Curve,Name="R_smile_pose")
    Control=RigVMVariableNode() {variable=R Corner Control}.Value
   )
   RigVMRerouteNode(Value=())
  ]}
]
hierarchy [
 Head {type=Connector, loc=<0 0 0>, rot=<0 0 0 1>}
 Root {type=Socket, loc=<0 0 0>, rot=<0 0 0 1>}
 BothLegIK {type=Curve, loc=<0 0 0>, rot=<0 0 0 1>}
 look_right_pose {type=Curve, loc=<0 0 0>, rot=<0 0 0 1>}
 L_wide_pose {type=Curve, loc=<0 0 0>, rot=<0 0 0 1>}
 R_lip_corner_narrow_pose {type=Curve, loc=<0 0 0>, rot=<0 0 0 1>}
 R_upper_lip_lower_pose {type=Curve, loc=<0 0 0>, rot=<0 0 0 1>}
 L_lip_corner_narrow_pose {type=Curve, loc=<0 0 0>, rot=<0 0 0 1>}
 L_upper_lip_lower_pose {type=Curve, loc=<0 0 0>, rot=<0 0 0 1>}
 R_lower_lip_up_pose {type=Curve, loc=<0 0 0>, rot=<0 0 0 1>}
 R_brow_squeeze_in_pose {type=Curve, loc=<0 0 0>, rot=<0 0 0 1>}
 L_lip_corner_wide_pose {type=Curve, loc=<0 0 0>, rot=<0 0 0 1>}
 L_upperLidDown_pose {type=Curve, loc=<0 0 0>, rot=<0 0 0 1>}
 R_lower_lip_down_pose {type=Curve, loc=<0 0 0>, rot=<0 0 0 1>}
 R_upper_lip_raiser_pose {type=Curve, loc=<0 0 0>, rot=<0 0 0 1>}
 L_lower_lip_down_pose {type=Curve, loc=<0 0 0>, rot=<0 0 0 1>}
 L_upper_lip_raiser_pose {type=Curve, loc=<0 0 0>, rot=<0 0 0 1>}
 L_brow_squeeze_in_pose {type=Curve, loc=<0 0 0>, rot=<0 0 0 1>}
 L_lower_lip_up_pose {type=Curve, loc=<0 0 0>, rot=<0 0 0 1>}
 R_upperLidDown_pose {type=Curve, loc=<0 0 0>, rot=<0 0 0 1>}
 R_wide_pose {type=Curve, loc=<0 0 0>, rot=<0 0 0 1>}
 R_blink_pose {type=Curve, loc=<0 0 0>, rot=<0 0 0 1>}
 look_down_pose {type=Curve, loc=<0 0 0>, rot=<0 0 0 1>}
 look_up_pose {type=Curve, loc=<0 0 0>, rot=<0 0 0 1>}
 L_blink_pose {type=Curve, loc=<0 0 0>, rot=<0 0 0 1>}
 Reaction_Happy {type=Curve, loc=<0 0 0>, rot=<0 0 0 1>}
 look_left_pose {type=Curve, loc=<0 0 0>, rot=<0 0 0 1>}
 R_lip_corner_wide_pose {type=Curve, loc=<0 0 0>, rot=<0 0 0 1>}
 L_smile_pose {type=Curve, loc=<0 0 0>, rot=<0 0 0 1>}
 R_smile_pose {type=Curve, loc=<0 0 0>, rot=<0 0 0 1>}
 L_frown_pose {type=Curve, loc=<0 0 0>, rot=<0 0 0 1>}
 R_frown_pose {type=Curve, loc=<0 0 0>, rot=<0 0 0 1>}
 C_nose_wrinkler_pose {type=Curve, loc=<0 0 0>, rot=<0 0 0 1>}
 L_cheek_up_pose {type=Curve, loc=<0 0 0>, rot=<0 0 0 1>}
 R_cheek_up_pose {type=Curve, loc=<0 0 0>, rot=<0 0 0 1>}
 jaw_open_pose {type=Curve, loc=<0 0 0>, rot=<0 0 0 1>}
 jaw_left_pose {type=Curve, loc=<0 0 0>, rot=<0 0 0 1>}
 jaw_right_pose {type=Curve, loc=<0 0 0>, rot=<0 0 0 1>}
 jaw_up_pose {type=Curve, loc=<0 0 0>, rot=<0 0 0 1>}
 jaw_forward_pose {type=Curve, loc=<0 0 0>, rot=<0 0 0 1>}
 jaw_back_pose {type=Curve, loc=<0 0 0>, rot=<0 0 0 1>}
 eye_convergence_pose {type=Curve, loc=<0 0 0>, rot=<0 0 0 1>}
 tongue_forward_pose {type=Curve, loc=<0 0 0>, rot=<0 0 0 1>}
 tongue_back_pose {type=Curve, loc=<0 0 0>, rot=<0 0 0 1>}
 tongue_up_pose {type=Curve, loc=<0 0 0>, rot=<0 0 0 1>}
 tongue_down_pose {type=Curve, loc=<0 0 0>, rot=<0 0 0 1>}
 L_brow_In_Up_pose {type=Curve, loc=<0 0 0>, rot=<0 0 0 1>}
 L_brow_Ot_Up_pose {type=Curve, loc=<0 0 0>, rot=<0 0 0 1>}
 L_brow_In_Dn_pose {type=Curve, loc=<0 0 0>, rot=<0 0 0 1>}
 L_brow_Ot_Dn_pose {type=Curve, loc=<0 0 0>, rot=<0 0 0 1>}
 R_brow_In_Up_pose {type=Curve, loc=<0 0 0>, rot=<0 0 0 1>}
 R_brow_Ot_Up_pose {type=Curve, loc=<0 0 0>, rot=<0 0 0 1>}
 R_brow_In_Dn_pose {type=Curve, loc=<0 0 0>, rot=<0 0 0 1>}
 R_brow_Ot_Dn_pose {type=Curve, loc=<0 0 0>, rot=<0 0 0 1>}
 upper_lip_up_pose {type=Curve, loc=<0 0 0>, rot=<0 0 0 1>}
 upper_lip_dn_pose {type=Curve, loc=<0 0 0>, rot=<0 0 0 1>}
 lower_lip_up_pose {type=Curve, loc=<0 0 0>, rot=<0 0 0 1>}
 lower_lip_dn_pose {type=Curve, loc=<0 0 0>, rot=<0 0 0 1>}
 L_smile_open_correction_pose {type=Curve, loc=<0 0 0>, rot=<0 0 0 1>}
 R_smile_open_correction_pose {type=Curve, loc=<0 0 0>, rot=<0 0 0 1>}
 tongue_left_pose {type=Curve, loc=<0 0 0>, rot=<0 0 0 1>}
 tongue_right_pose {type=Curve, loc=<0 0 0>, rot=<0 0 0 1>}
 reverse_jaw_pose {type=Curve, loc=<0 0 0>, rot=<0 0 0 1>}
 base_pose {type=Curve, loc=<0 0 0>, rot=<0 0 0 1>}
 L_eye_look_up_pose {type=Curve, loc=<0 0 0>, rot=<0 0 0 1>}
 L_eye_look_left_pose {type=Curve, loc=<0 0 0>, rot=<0 0 0 1>}
 L_eye_look_down_pose {type=Curve, loc=<0 0 0>, rot=<0 0 0 1>}
 L_eye_look_right_pose {type=Curve, loc=<0 0 0>, rot=<0 0 0 1>}
 R_eye_look_left_pose {type=Curve, loc=<0 0 0>, rot=<0 0 0 1>}
 R_eye_look_right_pose {type=Curve, loc=<0 0 0>, rot=<0 0 0 1>}
 R_eye_look_up_pose {type=Curve, loc=<0 0 0>, rot=<0 0 0 1>}
 R_eye_look_down_pose {type=Curve, loc=<0 0 0>, rot=<0 0 0 1>}
 L_lower_lid_up_pose {type=Curve, loc=<0 0 0>, rot=<0 0 0 1>}
 R_lower_lid_up_pose {type=Curve, loc=<0 0 0>, rot=<0 0 0 1>}
 Root {loc=<0 0 0>, rot=<0 0 0 1>}
 pelvis {parent=Root, loc=<0 -14.6653 38.9513>, rot=<0.5006 0.4994 -0.5006 -0.4994>}
 spine_01 {parent=pelvis, loc=<0 0 0>, rot=<0 0 -0.0124 -0.9999>}
 spine_02 {parent=spine_01, loc=<15.529 -0.3867 0>, rot=<0 0 0.0156 -0.9999>}
 spine_03 {parent=spine_02, loc=<14.3456 0 0>, rot=<0 0 0.004 -1>}
 neck_01 {parent=spine_03, loc=<0.7833 0.3499 0>, rot=<0 0 0.5128 -0.8585>}
 neck_02 {parent=neck_01, loc=<12.1968 0 0>, rot=<0 0 0.0313 -0.9995>}
 Head {parent=neck_02, loc=<9.6064 0 0>, rot=<-0.1386 -0.6934 0.1386 -0.6934>}
 face_attach {parent=Head, loc=<0 0 0>, rot=<0 0 0 1>}
 jaw {parent=face_attach, loc=<0 1.257 2.4215>, rot=<0 0.7071 0 -0.7071>}
 lip_lower_03_l {parent=jaw, loc=<-7.4496 7.0207 -4.261>, rot=<-0.0671 -0.7112 -0.0827 -0.6948>}
 lip_lower_02_l {parent=jaw, loc=<-7.2595 9.7274 -3.5139>, rot=<-0.1806 -0.7258 -0.1279 -0.6514>}
 lip_lower_01_l {parent=jaw, loc=<-6.9353 12.7416 -2.2731>, rot=<-0.1667 -0.7326 -0.1369 -0.6456>}
 lip_lower_mid {parent=jaw, loc=<-7.06 13.8318 0>, rot=<0.4939 0.506 0.506 0.4939>}
 lip_lower_03_r {parent=jaw, loc=<-7.4496 7.0207 4.261>, rot=<0.7112 -0.0671 -0.6948 0.0827>}
 lip_lower_01_r {parent=jaw, loc=<-6.9353 12.7417 2.2731>, rot=<0.7326 -0.1667 -0.6456 0.1369>}
 lip_lower_02_r {parent=jaw, loc=<-7.2595 9.7274 3.5139>, rot=<0.7258 -0.1806 -0.6514 0.1279>}
 teeth_lower {parent=jaw, loc=<-6.7086 6.2372 0>, rot=<0 -0.7071 0 -0.7071>}
 tongue_01 {parent=jaw, loc=<-4.6681 0.7455 0>, rot=<0.5253 -0.4734 0.4734 -0.5253>}
 tongue_02 {parent=tongue_01, loc=<-5.0324 -0.0138 0.1501>, rot=<-0.0051 0.0564 -0.0009 -0.9984>}
 tongue_03 {parent=tongue_02, loc=<-3.7171 0.0226 0.4386>, rot=<0.0051 -0.0111 0.0006 -0.9999>}
 lip_corner_l {parent=face_attach, loc=<4.6401 7.4463 -4.1217>, rot=<0.0187 -0.0115 -0.106 -0.9941>}
 cheek_bt_l {parent=face_attach, loc=<7.7862 2.8351 -3.0651>, rot=<0.0975 -0.0174 -0.1751 -0.9796>}
 cheek_bt_r {parent=face_attach, loc=<-7.7862 2.8351 -3.0651>, rot=<0.9796 -0.1751 0.0174 0.0975>}
 lip_corner_r {parent=face_attach, loc=<-4.6401 7.4463 -4.1217>, rot=<0.9941 -0.106 0.0115 0.0187>}
 reverse_jaw {parent=face_attach, loc=<0 2.7286 -3.1927>, rot=<0 0.7071 0 -0.7071>}
 lip_upper_03_r {parent=reverse_jaw, loc=<1.0041 5.9897 5.1335>, rot=<0.7112 -0.0671 -0.6948 0.0827>}
 lip_upper_02_r {parent=reverse_jaw, loc=<1.275 8.5527 4.2666>, rot=<0.7258 -0.1806 -0.6514 0.1279>}
 lip_upper_01_r {parent=reverse_jaw, loc=<1.3683 12.3382 3.134>, rot=<0.6929 -0.2403 -0.628 0.2602>}
 lip_upper_mid {parent=reverse_jaw, loc=<1.8989 14.219 0>, rot=<-0.4824 -0.517 -0.517 -0.4824>}
 lip_upper_01_l {parent=reverse_jaw, loc=<1.3684 12.3381 -3.1339>, rot=<-0.2403 -0.6929 -0.2602 -0.628>}
 lip_upper_02_l {parent=reverse_jaw, loc=<1.2751 8.5527 -4.2666>, rot=<-0.1806 -0.7258 -0.1279 -0.6514>}
 lip_upper_03_l {parent=reverse_jaw, loc=<1.0041 5.9897 -5.1335>, rot=<-0.0671 -0.7112 -0.0827 -0.6948>}
 cheek_tp_in_l {parent=reverse_jaw, loc=<5.7551 6.6606 -2.4082>, rot=<-0.2594 -0.6712 -0.2639 -0.6423>}
 cheek_tp_ot_l {parent=reverse_jaw, loc=<4.9733 2.7698 -7.0218>, rot=<-0.1358 -0.6932 -0.1692 -0.6874>}
 cheek_tp_in_r {parent=reverse_jaw, loc=<5.7551 6.6606 2.4082>, rot=<0.6712 -0.2594 -0.6423 0.2639>}
 cheek_tp_ot_r {parent=reverse_jaw, loc=<4.9733 2.7698 7.0218>, rot=<0.6932 -0.1358 -0.6874 0.1692>}
 brow_ot_r {parent=reverse_jaw, loc=<10.2921 2.3975 6.1241>, rot=<-0.6792 -0.1966 0.6792 0.1966>}
 brow_md_r {parent=reverse_jaw, loc=<11.6637 3.9036 4.1437>, rot=<-0.6803 -0.1929 0.6803 0.1929>}
 brow_ct {parent=reverse_jaw, loc=<11.7248 4.3074 0>, rot=<0.1061 -0.6991 0.1061 -0.6991>}
 brow_md_l {parent=reverse_jaw, loc=<11.6637 3.9036 -4.1437>, rot=<0.1929 -0.6803 0.1929 -0.6803>}
 brow_in_l {parent=reverse_jaw, loc=<12.0101 4.7652 -2.0511>, rot=<0.1061 -0.6991 0.1061 -0.6991>}
 brow_ot_l {parent=reverse_jaw, loc=<10.2921 2.3976 -6.1241>, rot=<0.1966 -0.6792 0.1966 -0.6792>}
 brow_in_r {parent=reverse_jaw, loc=<12.0101 4.7653 2.0511>, rot=<-0.6991 -0.1061 0.6991 0.1061>}
 eye_r {parent=reverse_jaw, loc=<8.4394 2.7735 3.4163>, rot=<0.693 0.1403 -0.6874 -0.1657>}
 eye_rSocket {parent=eye_r, type=Null, loc=<0 -2.1411 0>, rot=<0 0 0 1>}
 ear_01_r {parent=reverse_jaw, loc=<6.4278 -2.8708 1.4184>, rot=<-0.1188 -0.0182 0.9922 0.0334>}
 ear_02_r {parent=ear_01_r, loc=<-6.4266 1.2899 2.1608>, rot=<0 0 0 -1>}
 ear_03_r {parent=ear_02_r, loc=<-3.908 -0.0692 0.731>, rot=<0 0 0 -1>}
 ear_04_r {parent=ear_03_r, loc=<-5.0986 0 0>, rot=<0 0 0 -1>}
 ear_03_rSocket {parent=ear_03_r, type=Null, loc=<-4.968 0 -0.9058>, rot=<0 0 0 1>}
 ear_ot_r {parent=ear_02_r, loc=<-3.4353 0.115 2.9256>, rot=<0 0 0 1>}
 ear_in_r {parent=ear_02_r, loc=<-4.2352 -2.0162 -1.8546>, rot=<0 0 0 1>}
 ear_01_l {parent=reverse_jaw, loc=<6.4278 -2.8708 -1.4184>, rot=<0.0182 -0.1188 0.0334 -0.9922>}
 ear_02_l {parent=ear_01_l, loc=<6.4266 -1.2899 -2.1608>, rot=<0 0 0 1>}
 ear_03_l {parent=ear_02_l, loc=<3.908 0.0691 -0.731>, rot=<0 0 0 1>}
 ear_04_l {parent=ear_03_l, loc=<5.0986 0 0>, rot=<0 0 0 1>}
 ear_03_lSocket {parent=ear_03_l, type=Null, loc=<5.5072 0 0>, rot=<0 0 0 1>}
 ear_in_l {parent=ear_02_l, loc=<4.2352 2.0162 1.8546>, rot=<0 0 0 1>}
 ear_ot_l {parent=ear_02_l, loc=<3.4353 -0.115 -2.9256>, rot=<0 0 0 1>}
 eye_l {parent=reverse_jaw, loc=<8.4394 2.7736 -3.4163>, rot=<0.1403 -0.693 0.1657 -0.6874>}
 eye_lSocket {parent=eye_l, type=Null, loc=<0 2.0925 0>, rot=<0 0 0 1>}
 lid_bt_r {parent=reverse_jaw, loc=<8.4394 2.7735 3.4163>, rot=<-0.6903 -0.153 0.6903 0.153>}
 lid_tp_r {parent=reverse_jaw, loc=<8.4394 2.7735 3.4163>, rot=<-0.6903 -0.153 0.6903 0.153>}
 nose_mid {parent=reverse_jaw, loc=<4.0171 14.2245 0>, rot=<-0.546 -0.4493 -0.4493 -0.546>}
 nostril_l {parent=nose_mid, loc=<0.6716 -1.4444 -0.0035>, rot=<0 0 0 -1>}
 nostril_r {parent=nose_mid, loc=<0.6716 1.4444 -0.0035>, rot=<0 -1 0 0>}
 nose_midSocket {parent=nose_mid, type=Null, loc=<1.3482 0 0.7575>, rot=<0 0 0 1>}
 lid_tp_l {parent=reverse_jaw, loc=<8.4394 2.7736 -3.4163>, rot=<0.153 -0.6903 0.153 -0.6903>}
 lid_bt_l {parent=reverse_jaw, loc=<8.4394 2.7736 -3.4163>, rot=<0.153 -0.6903 0.153 -0.6903>}
 teeth_upper {parent=reverse_jaw, loc=<2.6971 5.3069 0>, rot=<0 -0.7071 0 -0.7071>}
 headSocket {parent=Head, type=Null, loc=<0 4.1427 10.8575>, rot=<0 0 0 1>}
 shoulder_fr_l {parent=spine_03, loc=<3.3143 1.7539 -7.578>, rot=<0.0042 0.003 -0.8375 -0.5464>}
 elbow_fr_l {parent=shoulder_fr_l, loc=<16.0381 0 0>, rot=<0.0002 0.0027 0.2036 -0.9791>}
 wrist_fr_l {parent=elbow_fr_l, loc=<15.7305 0 0>, rot=<0.9928 -0.1202 0 0>}
 ball_fr_l {parent=wrist_fr_l, loc=<4.0643 0 0>, rot=<0.0014 0.0007 -0.4114 -0.9115>}
 toe_fr_l {parent=ball_fr_l, loc=<2.6599 0 0>, rot=<-0.0027 -0.0108 -0.1114 -0.9937>}
 toe_tip_fr_l {parent=toe_fr_l, loc=<4.2353 0 0>, rot=<0.0035 0.0046 -0.1104 -0.9939>}
 ball_fr_lSocket {parent=ball_fr_l, type=Null, loc=<1.6306 0 0>, rot=<0 0 0 1>}
 scapula_fr_l {parent=shoulder_fr_l, loc=<-5.958 6.6503 -0.4062>, rot=<-0.0056 0.0274 0.4074 -0.9128>}
 shoulder_fr_r {parent=spine_03, loc=<3.3143 1.7539 7.578>, rot=<0.003 -0.0042 0.5464 -0.8375>}
 elbow_fr_r {parent=shoulder_fr_r, loc=<-16.0381 0 0>, rot=<0.0002 0.0027 0.2036 -0.9791>}
 wrist_fr_r {parent=elbow_fr_r, loc=<-15.7306 0.0001 0>, rot=<0.9928 -0.1202 0 0>}
 ball_fr_r {parent=wrist_fr_r, loc=<-4.0643 0.0001 0>, rot=<0.0014 0.0007 -0.4114 -0.9115>}
 toe_fr_r {parent=ball_fr_r, loc=<-2.6599 0 0>, rot=<-0.0027 -0.0108 -0.1114 -0.9937>}
 toe_tip_fr_r {parent=toe_fr_r, loc=<-4.2353 0 0>, rot=<0.0035 0.0046 -0.1104 -0.9939>}
 ball_fr_rSocket {parent=ball_fr_r, type=Null, loc=<-2.0884 0 0>, rot=<0 0 0 1>}
 scapula_fr_r {parent=shoulder_fr_r, loc=<5.9579 -6.6503 0.4062>, rot=<-0.0056 0.0274 0.4074 -0.9128>}
 attach_accessory {parent=spine_03, loc=<12.8014 -0.3993 0>, rot=<-0.497 -0.503 0.497 -0.503>}
 acc_fur {parent=attach_accessory, type=Null, loc=<0 1.45 0>, rot=<0 0 0 1>}
 acc_chonk {parent=attach_accessory, type=Null, loc=<0 0 0>, rot=<0 0 0 1>}
 acc_chonk_fur {parent=attach_accessory, type=Null, loc=<0 0 0>, rot=<0 0 0 1>}
 acc_clothes {parent=attach_accessory, type=Null, loc=<0 2.25 0>, rot=<0 0 0 1>}
 acc_clothes_chonk {parent=attach_accessory, type=Null, loc=<0 0 0>, rot=<0 0 0 1>}
 attach_backpack {parent=spine_02, loc=<14.3456 0 0>, rot=<0 0 0.004 -1>}
 attach_cape {parent=spine_02, loc=<14.3456 0 0>, rot=<0 0 0.004 -1>}
 hips {parent=pelvis, loc=<0 0 0>, rot=<0 0 0 1>}
 tail_01 {parent=hips, loc=<-10.1036 -4.046 0>, rot=<-0.0242 -0.7067 0.7067 0.0242>}
 tail_02 {parent=tail_01, loc=<8.8431 0 0>, rot=<0 -0.0045 0 -1>}
 tail_03 {parent=tail_02, loc=<8.8438 0 0>, rot=<0 0.0315 0 -0.9995>}
 tail_04 {parent=tail_03, loc=<8.8346 0 0>, rot=<0 0.0277 0 -0.9996>}
 tail_04Socket {parent=tail_04, type=Null, loc=<7.1555 0 2.8692>, rot=<0 0 0 1>}
 thigh_bk_l {parent=hips, loc=<-3.6025 0.5607 -6.2814>, rot=<-0.6817 -0.7316 0 0>}
 knee_bk_l {parent=thigh_bk_l, loc=<16.6492 0 0>, rot=<0 0 0.4212 -0.907>}
 ankle_bk_l {parent=knee_bk_l, loc=<12.9953 0 0>, rot=<0 0 -0.4271 -0.9042>}
 ball_bk_l {parent=ankle_bk_l, loc=<11.8273 0 0>, rot=<0 0 -0.5311 -0.8473>}
 toe_bk_l {parent=ball_bk_l, loc=<2.1609 0 0>, rot=<0 0 -0.0729 -0.9973>}
 toe_tip_bk_l {parent=toe_bk_l, loc=<4.4515 0 0>, rot=<0 0 -0.226 -0.9741>}
 ball_bk_lSocket {parent=ball_bk_l, type=Null, loc=<2.2878 0 0>, rot=<0 0 0 1>}
 thigh_bk_r {parent=hips, loc=<-3.6025 0.5607 6.2814>, rot=<0.7316 -0.6817 0 0>}
 knee_bk_r {parent=thigh_bk_r, loc=<-16.6492 -0.0001 0>, rot=<0 0 0.4212 -0.907>}
 ankle_bk_r {parent=knee_bk_r, loc=<-12.9954 0 0>, rot=<0 0 -0.4271 -0.9042>}
 ball_bk_r {parent=ankle_bk_r, loc=<-11.8273 0 0>, rot=<0 0 -0.5311 -0.8473>}
 toe_bk_r {parent=ball_bk_r, loc=<-2.1609 0 0>, rot=<0 0 -0.0729 -0.9973>}
 toe_tip_bk_r {parent=toe_bk_r, loc=<-4.4515 0 0>, rot=<0 0 -0.226 -0.9741>}
 ball_bk_rSocket {parent=ball_bk_r, type=Null, loc=<-2.4773 0 0>, rot=<0 0 0 1>}
 pelvisSocket {parent=pelvis, type=Null, loc=<0 -8.8042 0>, rot=<0 0 0 1>}
 IK_Foot_Rear_Root {parent=Root, loc=<0 -30.8181 0>, rot=<0.0444 0 0 -0.999>}
 IK_Foot_Rear_L {parent=IK_Foot_Rear_Root, loc=<6.2809 -0.096 2.3511>, rot=<-0.3842 -0.5936 -0.3842 -0.5936>}
 IK_Foot_Rear_R {parent=IK_Foot_Rear_Root, loc=<-6.2809 -0.096 2.3511>, rot=<0.5936 -0.3842 0.5936 -0.3842>}
 IK_Foot_Front_Root {parent=Root, loc=<0 3.5274 0>, rot=<0.047 0 -0.0001 -0.9989>}
 IK_Foot_Front_L {parent=IK_Foot_Front_Root, loc=<7.5169 9.5389 3.7395>, rot=<-0.3935 -0.5883 -0.3993 -0.5828>}
 IK_Foot_Front_R {parent=IK_Foot_Front_Root, loc=<-7.5141 9.5408 3.7399>, rot=<0.5827 -0.3994 0.5883 -0.3934>}
 attach_emote_03 {parent=Root, loc=<0 0 0>, rot=<0 0 0 1>}
 attach_emote_02 {parent=Root, loc=<0 0 0>, rot=<0 0 0 1>}
 attach_emote_01 {parent=Root, loc=<0 0 0>, rot=<0 0 0 1>}
 attach {parent=Root, loc=<0 0 0>, rot=<0 0 0 1>}
 rootSocket {parent=Root, type=Null, loc=<0 0 0>, rot=<0 0 0 1>}
 tongue_curl_tip_up_pose {type=Curve, loc=<0 0 0>, rot=<0 0 0 1>}
 tongue_curl_tip_down_pose {type=Curve, loc=<0 0 0>, rot=<0 0 0 1>}
 tongue_curl_mid_up_pose {type=Curve, loc=<0 0 0>, rot=<0 0 0 1>}
 tongue_curl_mid_down_pose {type=Curve, loc=<0 0 0>, rot=<0 0 0 1>}
 L_tongue_twist_pose {type=Curve, loc=<0 0 0>, rot=<0 0 0 1>}
 R_tongue_twist_pose {type=Curve, loc=<0 0 0>, rot=<0 0 0 1>}
 AttachEmote1InPelvisSpace {type=Curve, loc=<0 0 0>, rot=<0 0 0 1>}
 AttachEmote2InPelvisSpace {type=Curve, loc=<0 0 0>, rot=<0 0 0 1>}
 AttachEmote3InPelvisSpace {type=Curve, loc=<0 0 0>, rot=<0 0 0 1>}
 AnimateAcc {type=Curve, loc=<0 0 0>, rot=<0 0 0 1>}
 pupil_shape {type=Curve, loc=<0 0 0>, rot=<0 0 0 1>}
 pupil_size {type=Curve, loc=<0 0 0>, rot=<0 0 0 1>}
 bypass_poseAsset {type=Curve, loc=<0 0 0>, rot=<0 0 0 1>}
]
variables [
 BK Control {type=FRigElementKey, default=(Type=None,Name="")}
 Control Scale {type=double, default=1}
 Expose Pupil Controls {type=bool, default=False}
 Face Rig Control {type=FRigElementKey, default=(Type=None,Name="")}
 Face Rig Offset Transform {type=FTransform, default=(Rotation=(X=0,Y=0,Z=0,W=1),Translation=(X=0,Y=0,Z=-30),Scale3D=(X=1,Y=1,Z=1))}
 Jaw Main Control {type=FRigElementKey, default=(Type=None,Name="")}
 L Blink Control {type=FRigElementKey, default=(Type=None,Name="")}
 L Brow Control {type=FRigElementKey, default=(Type=None,Name="")}
 L Cheek Control {type=FRigElementKey, default=(Type=None,Name="")}
 L Corner Control {type=FRigElementKey, default=(Type=None,Name="")}
 L Eye Control {type=FRigElementKey, default=(Type=None,Name="")}
 L Low Lid {type=FRigElementKey, default=(Type=None,Name="")}
 L Lower Lip Control {type=FRigElementKey, default=(Type=None,Name="")}
 L Top Lid {type=FRigElementKey, default=(Type=None,Name="")}
 L Upper Lip Control {type=FRigElementKey, default=(Type=None,Name="")}
 Lower Lip Control {type=FRigElementKey, default=(Type=None,Name="")}
 Nose Control {type=FRigElementKey, default=(Type=None,Name="")}
 Pupil Shape Control {type=FRigElementKey, default=(Type=None,Name="")}
 Pupil Size Control {type=FRigElementKey, default=(Type=None,Name="")}
 R Blink Control {type=FRigElementKey, default=(Type=None,Name="")}
 R Brow Control {type=FRigElementKey, default=(Type=None,Name="")}
 R Cheek Control {type=FRigElementKey, default=(Type=None,Name="")}
 R Corner Control {type=FRigElementKey, default=(Type=None,Name="")}
 R Eye Control {type=FRigElementKey, default=(Type=None,Name="")}
 R Low Lid {type=FRigElementKey, default=(Type=None,Name="")}
 R Lower Lip Control {type=FRigElementKey, default=(Type=None,Name="")}
 R Top Lid {type=FRigElementKey, default=(Type=None,Name="")}
 R Upper Lip Control {type=FRigElementKey, default=(Type=None,Name="")}
 Reverse Jaw Control {type=FRigElementKey, default=(Type=None,Name="")}
 Squeeze Control {type=FRigElementKey, default=(Type=None,Name="")}
 Tongue Control {type=FRigElementKey, default=(Type=None,Name="")}
 Tongue Curl Control {type=FRigElementKey, default=(Type=None,Name="")}
 Tongue Curl Tip Control {type=FRigElementKey, default=(Type=None,Name="")}
 Tongue Fwd Control {type=FRigElementKey, default=(Type=None,Name="")}
 Tongue Twist Control {type=FRigElementKey, default=(Type=None,Name="")}
 Upper Lip Control {type=FRigElementKey, default=(Type=None,Name="")}
]
